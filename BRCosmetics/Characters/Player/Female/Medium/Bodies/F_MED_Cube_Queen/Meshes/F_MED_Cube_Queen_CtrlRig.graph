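# Control Rig for the F_MED_Cube_Queen female body cosmetic, making procedural corrections to skirt and loincloth bones from thigh pose.
# The only step wired to execution nudges dyn_skirt_far_left_master along Y, using the left thigh's pelvis-relative Euler angle through a clamped remap.
# Loincloth offset and rotation chains driven by both thighs, plus a two-thigh transform constraint, remain in the graph but are disconnected from execution.
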ControlRigBlueprint /BRCosmetics/Characters/Player/Female/Medium/Bodies/F_MED_Cube_Queen/Meshes/F_MED_Cube_Queen_CtrlRig
graphs [
 RigVMFunctionLibrary {links [], nodes []}
 RigVMModel {nodes [
   n1 = &RigVMFunction_MathQuaternionToEuler(RotationOrder=ZYX, Value=&RigUnit_GetRelativeBoneTransform(Bone=thigh_l, Space=pelvis).Transform.Rotation)
   &RigUnit_ModifyTransforms(
    ExecutePin=()
    ItemToModify=((Item=(Type=Bone,Name="dyn_loincloth_01"),Transform=(Rotation=(X=0,Y=0,Z=0,W=1),Translation=(X=-3,Y=0,Z=0),Scale3D=(X=1,Y=1,Z=1))))
    Weight=1
    WeightMinimum=0
    ItemToModify.0.Transform.Translation.X=&RigVMFunction_MathFloatRemap(SourceMaximum=-60, SourceMinimum=-45, TargetMaximum=0, TargetMinimum=0, bClamp=true, Value=@n1.Result.Z).Result
    ItemToModify.0.Transform.Rotation=/Script/RigVM.RigVMFunction_MathQuaternionFromRotator(Rotator=(Pitch=0,Yaw=0,Roll=50)) {is=struct RigVMUnitNode}.Result
   )
   n2 = &RigVMFunction_MathQuaternionToEuler(RotationOrder=ZYX, Value=&RigUnit_GetRelativeBoneTransform(Bone=thigh_r, Space=pelvis).Transform.Rotation)
   > &RigUnit_ModifyTransforms(
    ItemToModify=((Item=(Type=Bone,Name="dyn_loincloth_01"),Transform=(Rotation=(X=0,Y=0,Z=0,W=1),Translation=(X=3,Y=0,Z=0),Scale3D=(X=1,Y=1,Z=1))))
    Weight=1
    ItemToModify.0.Transform.Translation.X=&RigVMFunction_MathFloatRemap(SourceMaximum=90, SourceMinimum=120, TargetMaximum=0, TargetMinimum=0, bClamp=true, Value=@n2.Result.Z).Result
   )
   &RigUnit_ModifyTransforms(
    ExecutePin=()
    ItemToModify=((Item=(Type=Bone,Name="dyn_loincloth_master"),Transform=(Rotation=(X=0,Y=0,Z=0,W=1),Translation=(X=0,Y=0,Z=5),Scale3D=(X=1,Y=1,Z=1))))
    Weight=1
   )
   > &RigUnit_ModifyTransforms(
    ItemToModify=((Item=(Type=Bone,Name="dyn_loincloth_master"),Transform=(Rotation=(X=0,Y=0,Z=0,W=1),Translation=(X=0,Y=0,Z=0),Scale3D=(X=1,Y=1,Z=1))))
    Weight=1
    ItemToModify.0.Transform.Rotation=&RigVMFunction_MathQuaternionFromRotator(
     Rotator=(Pitch=0,Yaw=0,Roll=50)
     Rotator.Roll=&RigVMFunction_MathFloatRemap(SourceMaximum=90, SourceMinimum=120, TargetMaximum=-10, TargetMinimum=0, bClamp=true, Value=@n2.Result.Z).Result
    ).Result
   )
   RigUnit_BeginExecution = &RigUnit_BeginExecution()
   n3 = &RigUnit_GetRelativeBoneTransform(Bone=thigh_l, Space=pelvis)
   &RigUnit_ModifyTransforms(
    ItemToModify=((Item=(Type=Bone,Name="dyn_skirt_far_left_master"),Transform=(Rotation=(X=0,Y=0,Z=0,W=1),Translation=(X=0,Y=-5,Z=0),Scale3D=(X=1,Y=1,Z=1))))
    Weight=1
    ItemToModify.0.Transform.Translation.Y=&RigVMFunction_MathFloatRemap(
     SourceMaximum=40
     SourceMinimum=10
     TargetMaximum=-3
     TargetMinimum=0
     bClamp=true
     Value=&RigVMFunction_MathQuaternionToEuler(RotationOrder=ZYX, Value=@n3.Transform.Rotation).Result.Y
    ).Result
    ExecutePin=@RigUnit_BeginExecution.ExecutePin
   )
   &RigVMFunction_MathFloatRemap(SourceMaximum=-60, SourceMinimum=-45, TargetMaximum=10, TargetMinimum=0, bClamp=true, Value=@n1.Result.Z)
   &RigVMFunction_MathQuaternionFromRotator(Rotator=(Pitch=0,Yaw=0,Roll=0))
   &RigUnit_ModifyTransforms(
    ExecutePin=()
    ItemToModify=((Item=(Type=Bone,Name="dyn_loincloth_master"),Transform=(Rotation=(X=0,Y=0,Z=0,W=1),Translation=(X=0,Y=0,Z=2),Scale3D=(X=1,Y=1,Z=1))))
    Weight=1
   )
   /Script/ControlRig.RigUnit_TransformConstraintPerItem(
    BaseItem=(Type=Bone,Name="pelvis")
    BaseTransformSpace=GlobalSpace
    ExecutePin=()
    Item=(Type=Bone,Name="dyn_loincloth_01")
    Targets="((Transform=(Rotation=(X=0,Y=0,Z=0,W=1),Translation=(X=0,Y=0,Z=0),Scale3D=(X=1,Y=1,Z=1)),Weight=0.5,bMaintainOffset=True,Filter=(TranslationFilter=(bX=True,bY=True,bZ=True),RotationFilter=(bX=True,bY=True,bZ=True),ScaleFilter=(bX=True,bY=True,bZ=True))),(T ...(truncated)"
    bUseInitialTransforms=True
    Targets.0.Transform=&RigUnit_GetRelativeBoneTransform(Bone=thigh_l, Space=pelvis).Transform
    Targets.1.Transform=&RigUnit_GetRelativeBoneTransform(Bone=thigh_r, Space=pelvis).Transform
    BaseTransform=&RigUnit_GetRelativeBoneTransform(Bone=dyn_loincloth_01, Space=pelvis).Transform
   ) {is=struct RigVMUnitNode}
  ]}
]
hierarchy [
 root {loc=<0 0 0>, rot=<0 0 0 1>}
 pelvis {parent=root, loc=<0 -0.5301 98.0571>, rot=<0.0127 0.707 -0.0127 -0.707>}
 spine_01 {parent=pelvis, loc=<6.1208 -0.9328 0>, rot=<0 0 0.034 -0.9994>}
 spine_02 {parent=spine_01, loc=<6.5996 -0.2095 0>, rot=<0 0 0.0238 -0.9997>}
 spine_03 {parent=spine_02, loc=<6.4914 -0.1237 0>, rot=<0 0 0.0149 -0.9999>}
 spine_04 {parent=spine_03, loc=<6.9598 0.0794 0>, rot=<0 0 -0.0132 -0.9999>}
 spine_05 {parent=spine_04, loc=<6.9987 0 0>, rot=<0 0 -0.0015 -1>}
 clavicle_l {parent=spine_05, loc=<5.4831 -0.1053 -3.2946>, rot=<0.0465 -0.7289 -0.009 -0.6829>}
 upperarm_l {parent=clavicle_l, loc=<9.8658 -1.2136 -0.4105>, rot=<0.0164 -0.3245 -0.0269 -0.9454>}
 lowerarm_l {parent=upperarm_l, loc=<26.465 -0.8345 0.259>, rot=<0.0036 0.0161 -0.2298 -0.9731>}
 hand_l {parent=lowerarm_l, loc=<22.7355 -0.3203 -0.6105>, rot=<-0.6784 -0.0223 0.0735 -0.7307>}
 index_metacarpal_l {parent=hand_l, loc=<1.6535 0.0183 -2.96>, rot=<0.0158 -0.0124 -0.0173 -0.9996>}
 index_01_l {parent=index_metacarpal_l, loc=<8.3008 0.0357 -0.0524>, rot=<-0.0044 0.0485 0.0924 -0.9945>}
 index_02_l {parent=index_01_l, loc=<3.3587 -0.3004 0.0452>, rot=<-0.002 0.0027 0.3225 -0.9466>}
 index_03_l {parent=index_02_l, loc=<2.5578 -0.123 -0.0252>, rot=<-0.002 0.0067 0.2558 -0.9667>}
 middle_metacarpal_l {parent=hand_l, loc=<2.1856 -0.1725 -0.5984>, rot=<0.006 0.0168 -0.0329 -0.9993>}
 middle_01_l {parent=middle_metacarpal_l, loc=<7.3275 -0.2447 0.0129>, rot=<-0.0208 0.0042 0.0858 -0.9961>}
 middle_02_l {parent=middle_01_l, loc=<4.0123 0.1401 0.1175>, rot=<-0.0218 0.0258 0.3512 -0.9357>}
 middle_03_l {parent=middle_02_l, loc=<2.6077 0.0534 0.1373>, rot=<0.0082 0.0188 0.1708 -0.9851>}
 pinky_metacarpal_l {parent=hand_l, loc=<1.7616 -0.9753 3.3814>, rot=<-0.0397 0.0814 -0.069 -0.9935>}
 pinky_01_l {parent=pinky_metacarpal_l, loc=<7.1709 -0.4831 -0.1094>, rot=<-0.0654 -0.0035 0.2242 -0.9723>}
 pinky_02_l {parent=pinky_01_l, loc=<2.8274 0.0893 -0.1043>, rot=<0.0037 0.0011 0.2854 -0.9584>}
 pinky_03_l {parent=pinky_02_l, loc=<2.2046 0.0307 -0.0078>, rot=<-0.0002 0.0098 0.1445 -0.9895>}
 ring_metacarpal_l {parent=hand_l, loc=<2.2074 -0.3625 1.3638>, rot=<-0.0284 0.0722 -0.0458 -0.9959>}
 ring_01_l {parent=ring_metacarpal_l, loc=<7.0295 -0.6197 -0.1446>, rot=<-0.0489 -0.0252 0.1892 -0.9804>}
 ring_02_l {parent=ring_01_l, loc=<3.6174 0.4647 -0.1027>, rot=<0 0 0.2806 -0.9598>}
 ring_03_l {parent=ring_02_l, loc=<2.4152 0.114 0.0199>, rot=<0 0 0.1189 -0.9929>}
 thumb_01_l {parent=hand_l, loc=<1.3118 -0.3191 -2.459>, rot=<0.6546 -0.4042 0.0664 -0.6354>}
 thumb_02_l {parent=thumb_01_l, loc=<5.9336 -0.0448 0.0401>, rot=<-0.0037 0.0213 0.298 -0.9543>}
 thumb_03_l {parent=thumb_02_l, loc=<2.2223 0.3682 -0.0907>, rot=<0.0118 -0.013 0.0066 -0.9998>}
 lowerarm_twist_01_l {parent=lowerarm_l, loc=<18.316 0 0>, rot=<0 0 0 1>}
 lowerarm_twist_02_l {parent=lowerarm_l, loc=<10.6707 0 0>, rot=<0 0 0 1>}
 upperarm_twist_01_l {parent=upperarm_l, loc=<7.2135 -0.1134 0>, rot=<0 0 0 1>}
 upperarm_twist_02_l {parent=upperarm_l, loc=<19.7998 -0.1134 0>, rot=<0 0 0 1>}
 dyn_shoulder_piece_left_master {parent=clavicle_l, loc=<4.1085 0.2352 4.7444>, rot=<0.369 0.0462 0.6848 -0.6267>}
 dyn_shoulder_piece_left {parent=dyn_shoulder_piece_left_master, loc=<0 0 0>, rot=<0 0 0 1>}
 clavicle_r {parent=spine_05, loc=<5.4831 -0.1053 3.2946>, rot=<0.7289 0.0465 -0.6829 0.009>}
 upperarm_r {parent=clavicle_r, loc=<-9.8658 1.2136 0.4104>, rot=<0.0164 -0.3245 -0.0269 -0.9454>}
 lowerarm_r {parent=upperarm_r, loc=<-26.4651 0.8345 -0.259>, rot=<0.0036 0.0161 -0.2298 -0.9731>}
 hand_r {parent=lowerarm_r, loc=<-22.7356 0.3203 0.6106>, rot=<-0.6784 -0.0223 0.0735 -0.7307>}
 index_metacarpal_r {parent=hand_r, loc=<-1.6534 -0.0181 2.9601>, rot=<0.0158 -0.0124 -0.0173 -0.9996>}
 index_01_r {parent=index_metacarpal_r, loc=<-8.3008 -0.0357 0.0524>, rot=<-0.0044 0.0485 0.0924 -0.9945>}
 index_02_r {parent=index_01_r, loc=<-3.3589 0.3003 -0.0453>, rot=<-0.002 0.0027 0.3225 -0.9466>}
 index_03_r {parent=index_02_r, loc=<-2.5578 0.123 0.0252>, rot=<-0.002 0.0067 0.2558 -0.9667>}
 middle_metacarpal_r {parent=hand_r, loc=<-2.1856 0.1726 0.5984>, rot=<0.006 0.0168 -0.0329 -0.9993>}
 middle_01_r {parent=middle_metacarpal_r, loc=<-7.3273 0.2447 -0.0129>, rot=<-0.0208 0.0042 0.0858 -0.9961>}
 middle_02_r {parent=middle_01_r, loc=<-4.0124 -0.14 -0.1175>, rot=<-0.0218 0.0258 0.3512 -0.9357>}
 middle_03_r {parent=middle_02_r, loc=<-2.6077 -0.0533 -0.1373>, rot=<0.0082 0.0188 0.1708 -0.9851>}
 pinky_metacarpal_r {parent=hand_r, loc=<-1.7617 0.9753 -3.3814>, rot=<-0.0397 0.0814 -0.069 -0.9935>}
 pinky_01_r {parent=pinky_metacarpal_r, loc=<-7.1707 0.4832 0.1094>, rot=<-0.0654 -0.0035 0.2242 -0.9723>}
 pinky_02_r {parent=pinky_01_r, loc=<-2.8274 -0.0892 0.1043>, rot=<0.0037 0.0011 0.2854 -0.9584>}
 pinky_03_r {parent=pinky_02_r, loc=<-2.2045 -0.031 0.0078>, rot=<-0.0002 0.0098 0.1445 -0.9895>}
 ring_metacarpal_r {parent=hand_r, loc=<-2.2073 0.3627 -1.3638>, rot=<-0.0284 0.0722 -0.0458 -0.9959>}
 ring_01_r {parent=ring_metacarpal_r, loc=<-7.0297 0.6197 0.1446>, rot=<-0.0489 -0.0252 0.1892 -0.9804>}
 ring_02_r {parent=ring_01_r, loc=<-3.6171 -0.4648 0.1027>, rot=<0 0 0.2806 -0.9598>}
 ring_03_r {parent=ring_02_r, loc=<-2.4153 -0.1141 -0.0199>, rot=<0 0 0.1189 -0.9929>}
 thumb_01_r {parent=hand_r, loc=<-1.3119 0.3192 2.459>, rot=<0.6546 -0.4042 0.0664 -0.6354>}
 thumb_02_r {parent=thumb_01_r, loc=<-5.9336 0.0447 -0.0401>, rot=<-0.0037 0.0213 0.298 -0.9543>}
 thumb_03_r {parent=thumb_02_r, loc=<-2.2223 -0.3682 0.0907>, rot=<0.0118 -0.013 0.0066 -0.9998>}
 lowerarm_twist_01_r {parent=lowerarm_r, loc=<-18.3161 0 0>, rot=<0 0 0 1>}
 lowerarm_twist_02_r {parent=lowerarm_r, loc=<-10.6707 0 0>, rot=<0 0 0 1>}
 upperarm_twist_01_r {parent=upperarm_r, loc=<-7.2135 0.1134 0>, rot=<0 0 0 1>}
 upperarm_twist_02_r {parent=upperarm_r, loc=<-19.7998 0.1134 0>, rot=<0 0 0 1>}
 dyn_shoulder_piece_right_master {parent=clavicle_r, loc=<-4.1086 -0.2352 -4.7442>, rot=<0.6827 0.6394 -0.3465 0.0706>}
 dyn_shoulder_piece_right {parent=dyn_shoulder_piece_right_master, loc=<0 0 0>, rot=<0 0 0 1>}
 neck_01 {parent=spine_05, loc=<11.3138 0.1529 0>, rot=<0 0 -0.2003 -0.9797>}
 neck_02 {parent=neck_01, loc=<4.5627 0 0>, rot=<0 0 0.0504 -0.9987>}
 head {parent=neck_02, loc=<5.3914 0 0>, rot=<0 0 0.115 -0.9934>}
 thigh_l {parent=pelvis, loc=<-1.0855 -0.2109 -8.7053>, rot=<0.0009 0.0355 0.0155 -0.9993>}
 calf_l {parent=thigh_l, loc=<-42.5284 -0.0019 -0.7297>, rot=<0.0006 -0.0161 -0.0391 -0.9991>}
 calf_twist_01_l {parent=calf_l, loc=<-35.8483 -1.9197 -0.5794>, rot=<0.0237 0.0095 -0.0035 -0.9997>}
 calf_twist_02_l {parent=calf_l, loc=<-19.1546 -1.7943 -0.2638>, rot=<0.0237 0.0095 -0.0035 -0.9997>}
 foot_l {parent=calf_l, loc=<-46.7401 -2.5014 -0.592>, rot=<0.0427 -0.0238 0.0353 -0.9982>}
 ball_l {parent=foot_l, loc=<-3.5658 10.8066 -0.1544>, rot=<0.0001 -0.0001 -0.7063 -0.7079>}
 thigh_twist_01_l {parent=thigh_l, loc=<-27.952 0 0>, rot=<0.0013 0 0 -1>}
 thigh_r {parent=pelvis, loc=<-1.0853 -0.2109 8.7053>, rot=<0.0355 -0.0009 0.9993 0.0155>}
 calf_r {parent=thigh_r, loc=<42.5286 0.0019 0.7297>, rot=<0.0006 -0.0161 -0.0391 -0.9991>}
 calf_twist_01_r {parent=calf_r, loc=<35.8485 1.9197 0.5793>, rot=<0.0237 0.0095 -0.0035 -0.9997>}
 calf_twist_02_r {parent=calf_r, loc=<19.1546 1.7943 0.2638>, rot=<0.0237 0.0095 -0.0035 -0.9997>}
 foot_r {parent=calf_r, loc=<46.7403 2.5014 0.5919>, rot=<0.0427 -0.0238 0.0353 -0.9982>}
 ball_r {parent=foot_r, loc=<3.5658 -10.8066 0.1544>, rot=<0.0001 -0.0001 -0.7063 -0.7079>}
 thigh_twist_01_r {parent=thigh_r, loc=<27.952 0 0>, rot=<0.0013 0 0 -1>}
 dyn_skirt_mid_left_master {parent=pelvis, loc=<-4.8163 -9.4804 -12.6229>, rot=<0.0181 -0.9191 0.3691 -0.1368>}
 dyn_skirt_mid_left_01 {parent=dyn_skirt_mid_left_master, loc=<0 0 0>, rot=<0 0 0 -1>}
 dyn_skirt_mid_left_02 {parent=dyn_skirt_mid_left_01, loc=<13.554 0 0>, rot=<-0.0115 0.0008 -0.1122 -0.9936>}
 dyn_skirt_mid_left_03 {parent=dyn_skirt_mid_left_02, loc=<13.9373 0 0>, rot=<0.0313 0.004 -0.0117 -0.9994>}
 dyn_skirt_mid_left_04 {parent=dyn_skirt_mid_left_03, loc=<17.9161 0 0>, rot=<-0.0194 0.0016 0.1456 -0.9892>}
 dyn_skirt_inner_left_master {parent=pelvis, loc=<-4.6174 -13.1895 -4.8117>, rot=<0.0399 -0.7485 0.6535 -0.1054>}
 dyn_skirt_inner_left_01 {parent=dyn_skirt_inner_left_master, loc=<0 0 0>, rot=<0 0 0 1>}
 dyn_skirt_inner_left_02 {parent=dyn_skirt_inner_left_01, loc=<16.5046 0 0>, rot=<-0.0121 -0.0315 -0.1321 -0.9907>}
 dyn_skirt_inner_left_03 {parent=dyn_skirt_inner_left_02, loc=<16.5953 0 0>, rot=<0.0269 -0.008 0.0535 -0.9982>}
 dyn_skirt_inner_left_04 {parent=dyn_skirt_inner_left_03, loc=<15.0035 0 0>, rot=<0.1009 0.0239 0.1469 -0.9837>}
 dyn_skirt_far_left_master {parent=pelvis, loc=<-2.854 -3.5984 -15.8116>, rot=<-0.0482 -0.9117 0.3672 -0.178>}
 dyn_skirt_far_left_01 {parent=dyn_skirt_far_left_master, loc=<0 0 0>, rot=<0 0 0 -1>}
 dyn_skirt_far_left_02 {parent=dyn_skirt_far_left_01, loc=<14.3249 0 0>, rot=<0.073 0.0036 -0.0797 -0.9941>}
 dyn_skirt_far_left_03 {parent=dyn_skirt_far_left_02, loc=<17.5389 0 0>, rot=<0.062 -0.0019 0.0059 -0.9981>}
 dyn_skirt_far_left_04 {parent=dyn_skirt_far_left_03, loc=<16.7311 0 0>, rot=<-0.0317 -0.0119 0.1109 -0.9933>}
 dyn_skirt_mid_right_master {parent=pelvis, loc=<-4.8163 -9.4803 12.6229>, rot=<0.1359 -0.3313 0.9334 -0.0237>}
 dyn_skirt_mid_right_01 {parent=dyn_skirt_mid_right_master, loc=<0 0 0>, rot=<0 0 0 -1>}
 dyn_skirt_mid_right_02 {parent=dyn_skirt_mid_right_01, loc=<13.5541 0 0>, rot=<0.0935 0.0007 0.1122 -0.9893>}
 dyn_skirt_mid_right_03 {parent=dyn_skirt_mid_right_02, loc=<13.9374 0 0>, rot=<0.0177 0.0024 0.0121 -0.9998>}
 dyn_skirt_mid_right_04 {parent=dyn_skirt_mid_right_03, loc=<17.9161 0 0>, rot=<-0.0964 0.011 -0.1452 -0.9846>}
 dyn_skirt_inner_right_master {parent=pelvis, loc=<-4.6174 -13.1895 4.8117>, rot=<0.1042 -0.6305 0.7679 -0.0431>}
 dyn_skirt_inner_right_01 {parent=dyn_skirt_inner_right_master, loc=<0 0 0>, rot=<0 0 0 -1>}
 dyn_skirt_inner_right_02 {parent=dyn_skirt_inner_right_01, loc=<16.5046 0 0>, rot=<0.0018 -0.022 0.134 -0.9907>}
 dyn_skirt_inner_right_03 {parent=dyn_skirt_inner_right_02, loc=<16.5953 0 0>, rot=<-0.0177 -0.0118 -0.0528 -0.9984>}
 dyn_skirt_inner_right_04 {parent=dyn_skirt_inner_right_03, loc=<15.0035 0 0>, rot=<-0.0475 0.0226 -0.1471 -0.9877>}
 dyn_skirt_far_right_master {parent=pelvis, loc=<-2.854 -3.5984 15.8116>, rot=<0.183 -0.2363 0.954 0.0228>}
 dyn_skirt_far_right_01 {parent=dyn_skirt_far_right_master, loc=<0 0 0>, rot=<0 0 0 1>}
 dyn_skirt_far_right_02 {parent=dyn_skirt_far_right_01, loc=<14.325 0 0>, rot=<0.0277 0.0177 0.0778 -0.9964>}
 dyn_skirt_far_right_03 {parent=dyn_skirt_far_right_02, loc=<17.5389 0 0>, rot=<0.0202 -0.0019 -0.0059 -0.9998>}
 dyn_skirt_far_right_04 {parent=dyn_skirt_far_right_03, loc=<16.7311 0 0>, rot=<0.0001 -0.0058 -0.1114 -0.9938>}
 dyn_loincloth_master {parent=pelvis, loc=<2.4163 6.8342 0>, rot=<0 0.7071 0.7071 0>}
 dyn_loincloth_01 {parent=dyn_loincloth_master, loc=<0 0 0>, rot=<0 0 0 1>}
 dyn_loincloth_02 {parent=dyn_loincloth_01, loc=<15.9882 0 0>, rot=<0 0.0179 0 -0.9998>}
 dyn_loincloth_03 {parent=dyn_loincloth_02, loc=<13.5297 0 0>, rot=<0 0 -0.0041 -1>}
 dyn_loincloth_04 {parent=dyn_loincloth_03, loc=<11.1157 0 0>, rot=<0 0 0.0102 -0.9999>}
 dyn_loincloth_01_left {parent=dyn_loincloth_01, loc=<3.8833 -6.0864 -0.0471>, rot=<0.5089 -0.5089 -0.491 0.491>}
 dyn_loincloth_01_right {parent=dyn_loincloth_01, loc=<3.8833 6.199 -0.0471>, rot=<0.5089 -0.5089 -0.491 0.491>}
]
