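# Physics asset for the female Cube Queen player body, covering the pelvis, legs and dynamic skirt and loincloth bones.
# Capsules form the pelvis, legs and six four-segment skirt chains anchored to the pelvis with narrow swing and twist limits.
# Inner and mid skirt tips add a crosswise second capsule.
# The loincloth chains four thin boxes, favoring one swing axis, from a tiny master capsule whose pelvis joint is angularly free.
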PhysicsAsset /BRCosmetics/Characters/Player/Female/Medium/Bodies/F_MED_Cube_Queen/Meshes/F_MED_Cube_Queen_Physics
bodies {
 calf_l [{shapeType=Capsule, length=35.5498, radius=7.2079, center=<-25.3173 -1.6059 0.0003>, rotation=<rot 90 0 0>}]
 calf_r [{shapeType=Capsule, length=35.658, radius=7.2079, center=<26.1782 1.9591 0.0002>, rotation=<rot 90 0 0>}]
 dyn_loincloth_01 [{shapeType=Box, extentX=9.0009, extentY=14.3321, extentZ=0.7783, shapeName=dyn_loincloth_01_box, center=<8.3484 0 1.3956>}]
 dyn_loincloth_02 [{shapeType=Box, extentX=11.5612, extentY=10.4058, extentZ=0.5909, shapeName=dyn_loincloth_02_box, center=<6.8675 0 0.2691>}]
 dyn_loincloth_03 [{shapeType=Box, extentX=11.5612, extentY=12.3636, extentZ=0.9672, shapeName=dyn_loincloth_03_box, center=<6.8675 0 0>}]
 dyn_loincloth_04 [{shapeType=Box, extentX=11.5612, extentY=8.7743, extentZ=0.9171, shapeName=dyn_loincloth_04_box, center=<6.8675 0 0>}]
 dyn_loincloth_master [{shapeType=Capsule, length=0.505, radius=0.505, shapeName=dyn_loincloth_master_capsule, rotation=<rot 0 -90 0>}]
 dyn_skirt_far_left_01 [{shapeType=Capsule, length=11.3272, radius=3, shapeName=dyn_skirt_far_left_01_capsule, center=<8.9025 0 -0.0001>, rotation=<rot 90 0 0>}]
 dyn_skirt_far_left_02 [{shapeType=Capsule, length=11.3272, radius=3, shapeName=dyn_skirt_far_left_02_capsule, center=<8.9025 0 -0.0001>, rotation=<rot 90 0 0>}]
 dyn_skirt_far_left_03 [{shapeType=Capsule, length=11.3272, radius=3, center=<8.9025 0 -0.0001>, rotation=<rot 90 0 0>}]
 dyn_skirt_far_left_04 [{shapeType=Capsule, length=9.7874, radius=4.5398, shapeName=dyn_skirt_far_left_04_capsule, center=<8.9025 0 -0.0001>, rotation=<rot 90 0 0>}]
 dyn_skirt_far_right_01 [{shapeType=Capsule, length=11.3272, radius=3, shapeName=dyn_skirt_far_right_01_capsule, center=<8.9025 0 -0.0001>, rotation=<rot 90 0 0>}]
 dyn_skirt_far_right_02 [{shapeType=Capsule, length=11.3272, radius=3, shapeName=dyn_skirt_far_right_02_capsule, center=<8.9025 0 -0.0001>, rotation=<rot 90 0 0>}]
 dyn_skirt_far_right_03 [{shapeType=Capsule, length=11.3272, radius=3, center=<8.9025 0 -0.0001>, rotation=<rot 90 0 0>}]
 dyn_skirt_far_right_04 [{shapeType=Capsule, length=9.7874, radius=4.5398, shapeName=dyn_skirt_far_left_04_capsule, center=<8.9025 0 -0.0001>, rotation=<rot 90 0 0>}]
 dyn_skirt_inner_left_01 [{shapeType=Capsule, length=9.8318, radius=4.4954, shapeName=dyn_skirt_inner_left_01_capsule, center=<8.9025 0 -0.0001>, rotation=<rot 90 0 0>}]
 dyn_skirt_inner_left_02 [{shapeType=Capsule, length=9.8318, radius=4.4954, shapeName=dyn_skirt_inner_left_02_capsule, center=<8.9025 0 -0.0001>, rotation=<rot 90 0 0>}]
 dyn_skirt_inner_left_03 [{shapeType=Capsule, length=9.8318, radius=4.4954, center=<8.9025 0 -0.0001>, rotation=<rot 90 0 0>}]
 dyn_skirt_inner_left_04 [
  {
   shapeType=Capsule
   length=21.0268
   radius=4.4954
   shapeName=dyn_skirt_inner_left_04_capsule
   center=<15.4451 0 -0.0004>
   rotation=<rot 90 0 0>
  }
  {
   shapeType=Capsule
   length=8.3608
   radius=4.4954
   shapeName=dyn_skirt_inner_left_04_capsule
   center=<25.4451 -2.0177 -0.0004>
   rotation=<rot 0 0 90>
  }
 ]
 dyn_skirt_inner_right_01 [{
   shapeType=Capsule
   length=9.8318
   radius=4.4954
   shapeName=dyn_skirt_inner_right_01_capsule
   center=<8.9025 0 -0.0001>
   rotation=<rot 90 0 0>
  }]
 dyn_skirt_inner_right_02 [{
   shapeType=Capsule
   length=9.8318
   radius=4.4954
   shapeName=dyn_skirt_inner_right_02_capsule
   center=<8.9025 0 -0.0001>
   rotation=<rot 90 0 0>
  }]
 dyn_skirt_inner_right_03 [{shapeType=Capsule, length=9.8318, radius=4.4954, center=<8.9025 0 -0.0001>, rotation=<rot 90 0 0>}]
 dyn_skirt_inner_right_04 [
  {
   shapeType=Capsule
   length=21.0268
   radius=4.4954
   shapeName=dyn_skirt_inner_left_04_capsule
   center=<15.4451 0 -0.0004>
   rotation=<rot 90 0 0>
  }
  {
   shapeType=Capsule
   length=8.3608
   radius=4.4954
   shapeName=dyn_skirt_inner_left_04_capsule
   center=<25.4451 3.0201 -0.0004>
   rotation=<rot 0 0 90>
  }
 ]
 dyn_skirt_mid_left_01 [{shapeType=Capsule, length=11.3272, radius=3, shapeName=dyn_skirt_mid_left_01_capsule, center=<8.9025 0 -0.0001>, rotation=<rot 90 0 0>}]
 dyn_skirt_mid_left_02 [{shapeType=Capsule, length=11.3272, radius=3, shapeName=dyn_skirt_mid_left_02_capsule, center=<8.9025 0 -0.0001>, rotation=<rot 90 0 0>}]
 dyn_skirt_mid_left_03 [{shapeType=Capsule, length=16.0399, radius=3.7925, center=<11.6106 0 -0.0002>, rotation=<rot 90 0 0>}]
 dyn_skirt_mid_left_04 [
  {shapeType=Capsule, length=22.5222, radius=3, shapeName=dyn_skirt_mid_left_04_capsule, center=<15.4451 0 -0.0004>, rotation=<rot 90 0 0>}
  {shapeType=Capsule, length=9.8054, radius=3, shapeName=dyn_skirt_mid_left_04_capsule, center=<17.6643 0 -0.0004>, rotation=<rot 0 0 90>}
 ]
 dyn_skirt_mid_right_01 [{shapeType=Capsule, length=11.3272, radius=3, shapeName=dyn_skirt_mid_right_01_capsule, center=<8.9025 0 -0.0001>, rotation=<rot 90 0 0>}]
 dyn_skirt_mid_right_02 [{shapeType=Capsule, length=11.3272, radius=3, shapeName=dyn_skirt_mid_right_02_capsule, center=<8.9025 0 -0.0001>, rotation=<rot 90 0 0>}]
 dyn_skirt_mid_right_03 [{shapeType=Capsule, length=16.0399, radius=3.7925, center=<11.6106 0 -0.0002>, rotation=<rot 90 0 0>}]
 dyn_skirt_mid_right_04 [
  {shapeType=Capsule, length=22.5222, radius=3, shapeName=dyn_skirt_mid_right_04_capsule, center=<15.4451 0 -0.0004>, rotation=<rot 90 0 0>}
  {shapeType=Capsule, length=7.7324, radius=3, shapeName=dyn_skirt_mid_right_04_capsule, center=<18.5871 0 -0.0001>, rotation=<rot 0 0 90>}
 ]
 foot_l [{shapeType=Capsule, length=15.4207, radius=5.3313, center=<-4.72 7.1045 0.0002>, rotation=<rot 0 5.7255 89.9999>}]
 foot_r [{shapeType=Capsule, length=15.4207, radius=5.3313, center=<4.983 -7.6053 -0.0003>, rotation=<rot 0 5.7257 89.9999>}]
 pelvis [
  {shapeType=Capsule, length=13.1602, radius=7.0182}
  {shapeType=Capsule, length=12.6077, radius=7.5707, shapeName=pelvis_capsule, center=<-19.1562 -0.0002 -0.0001>}
 ]
 thigh_l [{shapeType=Capsule, length=29.844, radius=7.5015, center=<-22.3715 -0.0002 -0.0001>, rotation=<rot 90 0 0>}]
 thigh_r [{shapeType=Capsule, length=29.844, radius=7.5015, center=<22.5596 -0.0004 -0.0001>, rotation=<rot 90 0 0>}]
}
constraints [
 {bone1Name=thigh_l, bone2Name=pelvis, swing1LimitDegrees=45, swing2LimitDegrees=45, twistLimitDegrees=45}
 {bone1Name=calf_l, bone2Name=thigh_l, swing1LimitDegrees=45, swing2LimitDegrees=45, twistLimitDegrees=45}
 {bone1Name=foot_l, bone2Name=calf_l, swing1LimitDegrees=45, swing2LimitDegrees=45, twistLimitDegrees=45}
 {bone1Name=thigh_r, bone2Name=pelvis, swing1LimitDegrees=45, swing2LimitDegrees=45, twistLimitDegrees=45}
 {bone1Name=calf_r, bone2Name=thigh_r, swing1LimitDegrees=45, swing2LimitDegrees=45, twistLimitDegrees=45}
 {bone1Name=foot_r, bone2Name=calf_r, swing1LimitDegrees=45, swing2LimitDegrees=45, twistLimitDegrees=45}
 {bone1Name=dyn_skirt_mid_left_03, bone2Name=dyn_skirt_mid_left_02, swing1LimitDegrees=20, swing2LimitDegrees=20, twistLimitDegrees=7}
 {bone1Name=dyn_skirt_inner_left_03, bone2Name=dyn_skirt_inner_left_02, swing1LimitDegrees=20, swing2LimitDegrees=20, twistLimitDegrees=7}
 {bone1Name=dyn_skirt_far_left_03, bone2Name=dyn_skirt_far_left_02, swing1LimitDegrees=20, swing2LimitDegrees=20, twistLimitDegrees=7}
 {bone1Name=dyn_skirt_mid_right_03, bone2Name=dyn_skirt_mid_right_02, swing1LimitDegrees=20, swing2LimitDegrees=20, twistLimitDegrees=7}
 {bone1Name=dyn_skirt_inner_right_03, bone2Name=dyn_skirt_inner_right_02, swing1LimitDegrees=20, swing2LimitDegrees=20, twistLimitDegrees=7}
 {bone1Name=dyn_skirt_far_right_03, bone2Name=dyn_skirt_far_right_02, swing1LimitDegrees=20, swing2LimitDegrees=20, twistLimitDegrees=7}
 {bone1Name=dyn_loincloth_02, bone2Name=dyn_loincloth_01, swing1LimitDegrees=15, swing2LimitDegrees=60, twistLimitDegrees=15}
 {bone1Name=dyn_skirt_mid_left_02, bone2Name=dyn_skirt_mid_left_01, swing1LimitDegrees=20, swing2LimitDegrees=20, twistLimitDegrees=7}
 {bone1Name=dyn_skirt_mid_left_01, bone2Name=pelvis, swing1LimitDegrees=15, swing2LimitDegrees=10, twistLimitDegrees=5}
 {bone1Name=dyn_skirt_mid_left_04, bone2Name=dyn_skirt_mid_left_03, swing1LimitDegrees=20, swing2LimitDegrees=20, twistLimitDegrees=7}
 {bone1Name=dyn_skirt_inner_left_01, bone2Name=pelvis, swing1LimitDegrees=15, swing2LimitDegrees=10, twistLimitDegrees=5}
 {bone1Name=dyn_skirt_inner_left_02, bone2Name=dyn_skirt_inner_left_01, swing1LimitDegrees=20, swing2LimitDegrees=20, twistLimitDegrees=7}
 {bone1Name=dyn_skirt_inner_left_04, bone2Name=dyn_skirt_inner_left_03, swing1LimitDegrees=20, swing2LimitDegrees=20, twistLimitDegrees=7}
 {bone1Name=dyn_skirt_far_left_01, bone2Name=pelvis, swing1LimitDegrees=15, swing2LimitDegrees=10, twistLimitDegrees=5}
 {bone1Name=dyn_skirt_far_left_02, bone2Name=dyn_skirt_far_left_01, swing1LimitDegrees=20, swing2LimitDegrees=20, twistLimitDegrees=7}
 {bone1Name=dyn_skirt_far_left_04, bone2Name=dyn_skirt_far_left_03, swing1LimitDegrees=20, swing2LimitDegrees=20, twistLimitDegrees=7}
 {bone1Name=dyn_skirt_mid_right_04, bone2Name=dyn_skirt_mid_right_03, swing1LimitDegrees=20, swing2LimitDegrees=20, twistLimitDegrees=7}
 {bone1Name=dyn_skirt_mid_right_02, bone2Name=dyn_skirt_mid_right_01, swing1LimitDegrees=20, swing2LimitDegrees=20, twistLimitDegrees=7}
 {bone1Name=dyn_skirt_mid_right_01, bone2Name=pelvis, swing1LimitDegrees=15, swing2LimitDegrees=10, twistLimitDegrees=5}
 {bone1Name=dyn_skirt_inner_right_01, bone2Name=pelvis, swing1LimitDegrees=15, swing2LimitDegrees=10, twistLimitDegrees=5}
 {bone1Name=dyn_skirt_inner_right_02, bone2Name=dyn_skirt_inner_right_01, swing1LimitDegrees=20, swing2LimitDegrees=20, twistLimitDegrees=7}
 {bone1Name=dyn_skirt_inner_right_04, bone2Name=dyn_skirt_inner_right_03, swing1LimitDegrees=20, swing2LimitDegrees=20, twistLimitDegrees=7}
 {bone1Name=dyn_skirt_far_right_01, bone2Name=pelvis, swing1LimitDegrees=15, swing2LimitDegrees=10, twistLimitDegrees=5}
 {bone1Name=dyn_skirt_far_right_02, bone2Name=dyn_skirt_far_right_01, swing1LimitDegrees=20, swing2LimitDegrees=20, twistLimitDegrees=7}
 {bone1Name=dyn_skirt_far_right_04, bone2Name=dyn_skirt_far_right_03, swing1LimitDegrees=20, swing2LimitDegrees=20, twistLimitDegrees=7}
 {bone1Name=dyn_loincloth_01, bone2Name=dyn_loincloth_master, swing1LimitDegrees=15, swing2LimitDegrees=60, twistLimitDegrees=15}
 {bone1Name=dyn_loincloth_03, bone2Name=dyn_loincloth_02, swing1LimitDegrees=15, swing2LimitDegrees=60, twistLimitDegrees=15}
 {bone1Name=dyn_loincloth_04, bone2Name=dyn_loincloth_03, swing1LimitDegrees=15, swing2LimitDegrees=60, twistLimitDegrees=15}
 {
  bone1Name=dyn_loincloth_master
  bone2Name=pelvis
  swing1LimitDegrees=45
  swing1Motion=Free
  swing2LimitDegrees=45
  swing2Motion=Free
  twistLimitDegrees=45
  twistMotion=Free
 }
]
