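# Cube Queen body AnimBP that adds skirt, loincloth and shoulder-piece secondary motion.
# Standing and crouch branches copy the parent mesh pose, apply weighted bone constraints, offsets and RigidBody physics, with a Chugga physics variant and an extra crouch Control Rig.
# They blend on crouch state into a deform Control Rig, then alpha-driven shoulder offsets.
# The EventGraph updates gravity override, detects Chugga montages and resets dynamics around crouching and emotes.
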AnimBlueprint /BRCosmetics/Characters/Player/Female/Medium/Bodies/F_MED_Cube_Queen/Meshes/F_MED_Cube_Queen_AnimBP
target_skeleton=&/F_MED_Cube_Queen_Skeleton
graphs [
 AnimGraph {nodes [
   "Constraint - dyn_skirt_far_left_master (PPPPP)"!(
    ComponentPose
    Alpha=1
    ConstraintWeights_0=0
    ConstraintWeights_1:float=0.4803
    ConstraintWeights_2:float=0.457
    ConstraintWeights_3:float=0.0001
    ConstraintWeights_4:float=0.0626
   ) -> (Pose) {
    alphaBoolBlend {blendOption=Linear}
    boneToModify {boneName=dyn_skirt_far_left_master}
    constraintSetup [
     {targetBone {boneName=thigh_r}}
     {targetBone {boneName=pelvis}}
     {targetBone {boneName=thigh_twist_01_l}}
     {targetBone {boneName=thigh_twist_01_r}}
     {targetBone {boneName=thigh_l}}
    ]
    constraintWeights [1,1,1,1,1]
   }
   n1 = RigidBody!(ComponentPose=Local To Component(LocalPose=Use cached pose 'ChuggaCache'()), Alpha, OverrideWorldGravity:Vector="0,0,-980", ExternalForce) -> (Pose) {
    bClampLinearTranslationLimitToRefPose=true
    bOverrideWorldGravity=true
    overridePhysicsAsset=&/F_MED_Cube_Queen_Physics_Chugga
    simulationSpace=BaseBoneSpace
    alphaBoolBlend {blendOption=Linear}
    baseBoneRef {boneName=root}
    componentAppliedLinearAccClamp=<4000 4000 1500>
    componentLinearAccScale=<0.45 0.45 0.45>
    componentLinearVelScale=<0.45 0.45 0.45>
    overrideWorldGravity {z=-980}
   }
   n2 = Component To Local(ComponentPose=@n1)
   n3 = "Transform (Modify) Bone - Bone: dyn_loincloth_master"(ComponentPose=Local To Component(LocalPose=Use cached pose 'ChuggaCache'()), Alpha=1, Translation="0,0,2", Rotation="0,0,0", Scale="1,1,1") {
    translationMode=BMM_Additive
    translationSpace=BCS_BoneSpace
    alphaBoolBlend {blendOption=Linear}
    boneToModify {boneName=dyn_loincloth_master}
    translation {z=2}
   }
   n4 = RigidBody(ComponentPose=@n3, OverrideWorldGravity=Get Gravity Override()) {
    bClampLinearTranslationLimitToRefPose=true
    bOverrideWorldGravity=true
    overridePhysicsAsset=&/F_MED_Cube_Queen_Physics
    simulationSpace=BaseBoneSpace
    alphaBoolBlend {blendOption=Linear}
    baseBoneRef {boneName=root}
    componentAppliedLinearAccClamp=<4000 4000 1500>
    componentLinearAccScale=<0.45 0.45 0.45>
    componentLinearVelScale=<0.45 0.45 0.45>
    overrideWorldGravity {z=-980}
   }
   n5 = Component To Local(ComponentPose=@n4)
   n6 = "Transform (Modify) Bone - Bone: dyn_skirt_far_right_master"(
    ComponentPose=Local To Component(LocalPose=Blend Poses by bool(BlendPose_0=@n2, BlendPose_1=@n5) {blendType=Linear})
    Alpha=1
    Translation="-1,0,1"
    Rotation="0,0,0"
    Scale="1,1,1"
   ) {
    rotationMode=BMM_Additive
    translationMode=BMM_Additive
    alphaBoolBlend {blendOption=Linear}
    boneToModify {boneName=dyn_skirt_far_right_master}
    translation {x=-1, z=1}
   }
   n7 = "Transform (Modify) Bone - Bone: dyn_skirt_far_left_master"(ComponentPose=@n6, Alpha=1, Translation="2,0,1", Rotation="0,0,0", Scale="1,1,1") {
    rotationMode=BMM_Additive
    translationMode=BMM_Additive
    alphaBoolBlend {blendOption=Linear}
    boneToModify {boneName=dyn_skirt_far_left_master}
    translation {x=2, z=1}
   }
   n8 = "Transform (Modify) Bone - Bone: dyn_skirt_far_right_master"(ComponentPose=@n7, Alpha=1, Translation="0,0,1", Rotation="0,0,0", Scale="1,1,1") {
    rotationMode=BMM_Additive
    rotationSpace=BCS_BoneSpace
    translationMode=BMM_Additive
    translationSpace=BCS_BoneSpace
    alphaBoolBlend {blendOption=Linear}
    boneToModify {boneName=dyn_skirt_far_right_master}
    translation {z=1}
   }
   n9 = "Constraint - dyn_loincloth_master (PP)"!(
    ComponentPose=Local To Component(LocalPose=Copy Pose From Mesh() {bUseAttachedParent=true})
    Alpha=Get CrouchMovingAlpha() -> (CrouchMovingAlpha:float)
    ConstraintWeights_0=0.5
    ConstraintWeights_1:float=0.5
   ) -> (Pose) {
    alphaBoolBlend {blendInTime=0.6, blendOption=Linear, blendOutTime=0.6}
    boneToModify {boneName=dyn_loincloth_master}
    constraintSetup [{targetBone {boneName=thigh_l}}, {targetBone {boneName=thigh_r}}]
    constraintWeights [1,1]
   }
   n10 = "Transform (Modify) Bone - Bone: dyn_loincloth_master"!(
    ComponentPose=@n9
    bAlphaBoolEnabled:bool=Get bIsCrouchMoving() -> (bIsCrouchMoving:bool)
    Translation="0,0,3"
    Rotation="0,0,-30"
    Scale="1,1,1"
   ) -> (Pose) {
    alphaInputType=Bool
    rotationMode=BMM_Additive
    scaleMode=BMM_Additive
    translationMode=BMM_Additive
    alphaBoolBlend {blendInTime=0.6, blendOption=Linear, blendOutTime=0.6}
    boneToModify {boneName=dyn_loincloth_master}
    rotation {roll=-30}
    translation {z=3}
   }
   n11 = "Transform (Modify) Bone - Bone: dyn_loincloth_master"(
    ComponentPose=@n10
    bAlphaBoolEnabled=Get bIsCrouchIdling() -> (bIsCrouchIdling:bool)
    Translation="0,0,3"
    Rotation="0,30,-30"
    Scale="1,1,1"
   ) {
    alphaInputType=Bool
    rotationMode=BMM_Additive
    scaleMode=BMM_Additive
    translationMode=BMM_Additive
    alphaBoolBlend {blendInTime=0.6, blendOption=Linear, blendOutTime=0.6}
    boneToModify {boneName=dyn_loincloth_master}
    rotation {roll=-30, yaw=30}
    translation {z=3}
   }
   n12 = "Transform (Modify) Bone - Bone: dyn_sim_space_01"!(ComponentPose=@n11, bAlphaBoolEnabled:bool=True, Translation="0,-50,0", Rotation="0,0,0", Scale="1,1,1") -> (Pose) {
    alphaInputType=Bool
    rotationMode=BMM_Additive
    scaleMode=BMM_Additive
    translationMode=BMM_Additive
    alphaBoolBlend {blendInTime=0.5, blendOption=Linear, blendOutTime=0.2}
    boneToModify {boneName=dyn_sim_space_01}
    translation {y=-50}
   }
   n13 = "Constraint - dyn_skirt_far_left_master (PPPPP)"(
    ComponentPose=@n12
    Alpha=1
    ConstraintWeights_0=0
    ConstraintWeights_1=0.3605
    ConstraintWeights_2=0.533
    ConstraintWeights_3=0
    ConstraintWeights_4=0.1064
   ) {
    alphaBoolBlend {blendOption=Linear}
    boneToModify {boneName=dyn_skirt_far_left_master}
    constraintSetup [
     {targetBone {boneName=thigh_r}}
     {targetBone {boneName=pelvis}}
     {targetBone {boneName=thigh_twist_01_l}}
     {targetBone {boneName=thigh_twist_01_r}}
     {targetBone {boneName=thigh_l}}
    ]
    constraintWeights [1,1,1,1,1]
   }
   n14 = "Constraint - dyn_skirt_mid_left_master (PPPPP)"!(
    ComponentPose=@n13
    Alpha=1
    ConstraintWeights_0=0.3699
    ConstraintWeights_1:float=0.1508
    ConstraintWeights_2:float=0.0036
    ConstraintWeights_3:float=0.4679
    ConstraintWeights_4:float=0.0077
   ) -> (Pose) {
    alphaBoolBlend {blendOption=Linear}
    boneToModify {boneName=dyn_skirt_mid_left_master}
    constraintSetup [
     {targetBone {boneName=pelvis}}
     {targetBone {boneName=thigh_twist_01_l}}
     {targetBone {boneName=dyn_skirt_far_left_01}}
     {targetBone {boneName=dyn_skirt_mid_left_01}}
     {targetBone {boneName=thigh_l}}
    ]
    constraintWeights [1,1,1,1,1]
   }
   n15 = "Constraint - dyn_skirt_inner_left_master (PPPPPPP)"!(
    ComponentPose=@n14
    Alpha=1
    ConstraintWeights_0=0
    ConstraintWeights_1:float=0.0023
    ConstraintWeights_2:float=0.0831
    ConstraintWeights_3:float=0.0014
    ConstraintWeights_4:float=0.3001
    ConstraintWeights_5:float=0.6118
    ConstraintWeights_6:float=0.0012
   ) -> (Pose) {
    alphaBoolBlend {blendOption=Linear}
    boneToModify {boneName=dyn_skirt_inner_left_master}
    constraintSetup [
     {targetBone {boneName=thigh_r}}
     {targetBone {boneName=thigh_twist_01_r}}
     {targetBone {boneName=thigh_twist_01_l}}
     {targetBone {boneName=thigh_l}}
     {targetBone {boneName=dyn_skirt_inner_left_01}}
     {targetBone {boneName=pelvis}}
     {targetBone {boneName=dyn_skirt_inner_right_01}}
    ]
    constraintWeights [1,1,1,1,1,1,1]
   }
   n16 = "Constraint - dyn_skirt_far_right_master (PPPPR)"!(
    ComponentPose=@n15
    Alpha=0.5
    ConstraintWeights_0=0.0835
    ConstraintWeights_1:float=0.4244
    ConstraintWeights_2:float=0.0001
    ConstraintWeights_3:float=0.4921
    ConstraintWeights_4:float=0
   ) -> (Pose) {
    alphaBoolBlend {blendOption=Linear}
    boneToModify {boneName=dyn_skirt_far_right_master}
    constraintSetup [
     {targetBone {boneName=thigh_r}}
     {targetBone {boneName=pelvis}}
     {targetBone {boneName=thigh_twist_01_l}}
     {targetBone {boneName=thigh_twist_01_r}}
     {transformType=Rotation, targetBone {boneName=thigh_l}}
    ]
    constraintWeights [1,1,1,1,1]
   }
   n17 = "Constraint - dyn_skirt_mid_right_master (PPPPP)"!(
    ComponentPose=@n16
    Alpha=1
    ConstraintWeights_0=0.0077
    ConstraintWeights_1:float=0.3699
    ConstraintWeights_2:float=0.4679
    ConstraintWeights_3:float=0.1508
    ConstraintWeights_4:float=0.0036
   ) -> (Pose) {
    alphaBoolBlend {blendOption=Linear}
    boneToModify {boneName=dyn_skirt_mid_right_master}
    constraintSetup [
     {targetBone {boneName=thigh_r}}
     {targetBone {boneName=pelvis}}
     {targetBone {boneName=dyn_skirt_mid_right_01}}
     {targetBone {boneName=thigh_twist_01_r}}
     {targetBone {boneName=dyn_skirt_far_right_01}}
    ]
    constraintWeights [1,1,1,1,1]
   }
   n18 = "Constraint - dyn_skirt_inner_right_master (PPPPPPP)"!(
    ComponentPose=@n17
    Alpha=1
    ConstraintWeights_0=0.0014
    ConstraintWeights_1:float=0.0831
    ConstraintWeights_2:float=0.0023
    ConstraintWeights_3:float=0
    ConstraintWeights_4:float=0.0012
    ConstraintWeights_5:float=0.6118
    ConstraintWeights_6:float=0.3001
   ) -> (Pose) {
    alphaBoolBlend {blendOption=Linear}
    boneToModify {boneName=dyn_skirt_inner_right_master}
    constraintSetup [
     {targetBone {boneName=thigh_r}}
     {targetBone {boneName=thigh_twist_01_r}}
     {targetBone {boneName=thigh_twist_01_l}}
     {targetBone {boneName=thigh_l}}
     {targetBone {boneName=dyn_skirt_inner_left_01}}
     {targetBone {boneName=pelvis}}
     {targetBone {boneName=dyn_skirt_inner_right_01}}
    ]
    constraintWeights [1,1,1,1,1,1,1]
   }
   n19 = RigidBody(ComponentPose=@n18, Alpha=Get CrouchSprintingAlpha() -> (CrouchSprintingAlpha:float), OverrideWorldGravity="0,0,-1100") {
    bClampLinearTranslationLimitToRefPose=true
    bOverrideWorldGravity=true
    overridePhysicsAsset=&/F_MED_Cube_Queen_Physics_Crouch
    simulationSpace=BaseBoneSpace
    alphaBoolBlend {blendOption=Linear}
    baseBoneRef {boneName=spine_01}
    componentAppliedLinearAccClamp=<2000 2000 2000>
    componentLinearAccScale=<0.25 0.25 0.25>
    componentLinearVelScale=<0.25 0.25 0.25>
    overrideWorldGravity {z=-980}
   }
   n20 = "Transform (Modify) Bone - Bone: dyn_loincloth_01_left"!(ComponentPose=@n19, bAlphaBoolEnabled:bool=True, Translation="0,0,0", Rotation="0,0,0", Scale="1,1,1") -> (Pose) {
    alphaInputType=Bool
    rotationMode=BMM_Additive
    scaleMode=BMM_Additive
    translationMode=BMM_Additive
    alphaBoolBlend {blendInTime=0.2, blendOption=Linear, blendOutTime=0.2}
    boneToModify {boneName=dyn_loincloth_01_left}
   }
   n21 = "Constraint - dyn_skirt_far_left_master (PPPPP)"(
    ComponentPose=@n20
    Alpha=1
    ConstraintWeights_0=0
    ConstraintWeights_1=0.3605
    ConstraintWeights_2=0.533
    ConstraintWeights_3=0
    ConstraintWeights_4=0.1064
   ) {
    alphaBoolBlend {blendOption=Linear}
    boneToModify {boneName=dyn_skirt_far_left_master}
    constraintSetup [
     {targetBone {boneName=thigh_r}}
     {targetBone {boneName=pelvis}}
     {targetBone {boneName=thigh_twist_01_l}}
     {targetBone {boneName=thigh_twist_01_r}}
     {targetBone {boneName=thigh_l}}
    ]
    constraintWeights [1,1,1,1,1]
   }
   n22 = "Constraint - dyn_skirt_mid_left_master (PPPPP)"(
    ComponentPose=@n21
    Alpha=1
    ConstraintWeights_0=0.3699
    ConstraintWeights_1=0.1508
    ConstraintWeights_2=0.0036
    ConstraintWeights_3=0.4679
    ConstraintWeights_4=0.0077
   ) {
    alphaBoolBlend {blendOption=Linear}
    boneToModify {boneName=dyn_skirt_mid_left_master}
    constraintSetup [
     {targetBone {boneName=pelvis}}
     {targetBone {boneName=thigh_twist_01_l}}
     {targetBone {boneName=dyn_skirt_far_left_01}}
     {targetBone {boneName=dyn_skirt_mid_left_01}}
     {targetBone {boneName=thigh_l}}
    ]
    constraintWeights [1,1,1,1,1]
   }
   n23 = "Constraint - dyn_skirt_inner_left_master (PPPPPPP)"(
    ComponentPose=@n22
    Alpha=1
    ConstraintWeights_0=0
    ConstraintWeights_1=0.0023
    ConstraintWeights_2=0.0831
    ConstraintWeights_3=0.0014
    ConstraintWeights_4=0.3001
    ConstraintWeights_5=0.6118
    ConstraintWeights_6=0.0012
   ) {
    alphaBoolBlend {blendOption=Linear}
    boneToModify {boneName=dyn_skirt_inner_left_master}
    constraintSetup [
     {targetBone {boneName=thigh_r}}
     {targetBone {boneName=thigh_twist_01_r}}
     {targetBone {boneName=thigh_twist_01_l}}
     {targetBone {boneName=thigh_l}}
     {targetBone {boneName=dyn_skirt_inner_left_01}}
     {targetBone {boneName=pelvis}}
     {targetBone {boneName=dyn_skirt_inner_right_01}}
    ]
    constraintWeights [1,1,1,1,1,1,1]
   }
   n24 = "Constraint - dyn_skirt_far_right_master (PPPPR)"(
    ComponentPose=@n23
    Alpha=0.5
    ConstraintWeights_0=0.0835
    ConstraintWeights_1=0.4244
    ConstraintWeights_2=0.0001
    ConstraintWeights_3=0.4921
    ConstraintWeights_4=0
   ) {
    alphaBoolBlend {blendOption=Linear}
    boneToModify {boneName=dyn_skirt_far_right_master}
    constraintSetup [
     {targetBone {boneName=thigh_r}}
     {targetBone {boneName=pelvis}}
     {targetBone {boneName=thigh_twist_01_l}}
     {targetBone {boneName=thigh_twist_01_r}}
     {transformType=Rotation, targetBone {boneName=thigh_l}}
    ]
    constraintWeights [1,1,1,1,1]
   }
   n25 = "Constraint - dyn_skirt_mid_right_master (PPPPP)"(
    ComponentPose=@n24
    Alpha=1
    ConstraintWeights_0=0.0077
    ConstraintWeights_1=0.3699
    ConstraintWeights_2=0.4679
    ConstraintWeights_3=0.1508
    ConstraintWeights_4=0.0036
   ) {
    alphaBoolBlend {blendOption=Linear}
    boneToModify {boneName=dyn_skirt_mid_right_master}
    constraintSetup [
     {targetBone {boneName=thigh_r}}
     {targetBone {boneName=pelvis}}
     {targetBone {boneName=dyn_skirt_mid_right_01}}
     {targetBone {boneName=thigh_twist_01_r}}
     {targetBone {boneName=dyn_skirt_far_right_01}}
    ]
    constraintWeights [1,1,1,1,1]
   }
   n26 = "Constraint - dyn_skirt_inner_right_master (PPPPPPP)"(
    ComponentPose=@n25
    Alpha=1
    ConstraintWeights_0=0.0014
    ConstraintWeights_1=0.0831
    ConstraintWeights_2=0.0023
    ConstraintWeights_3=0
    ConstraintWeights_4=0.0012
    ConstraintWeights_5=0.6118
    ConstraintWeights_6=0.3001
   ) {
    alphaBoolBlend {blendOption=Linear}
    boneToModify {boneName=dyn_skirt_inner_right_master}
    constraintSetup [
     {targetBone {boneName=thigh_r}}
     {targetBone {boneName=thigh_twist_01_r}}
     {targetBone {boneName=thigh_twist_01_l}}
     {targetBone {boneName=thigh_l}}
     {targetBone {boneName=dyn_skirt_inner_left_01}}
     {targetBone {boneName=pelvis}}
     {targetBone {boneName=dyn_skirt_inner_right_01}}
    ]
    constraintWeights [1,1,1,1,1,1,1]
   }
   n27 = "Transform (Modify) Bone - Bone: dyn_skirt_far_right_master"(ComponentPose=@n26, Alpha=1, Translation="0,1,1.5", Rotation="0,0,0", Scale="1,1,1") {
    rotationMode=BMM_Additive
    translationMode=BMM_Additive
    translationSpace=BCS_BoneSpace
    alphaBoolBlend {blendOption=Linear}
    boneToModify {boneName=dyn_skirt_far_right_master}
    translation {y=1, z=1.5}
   }
   n28 = "Transform (Modify) Bone - Bone: dyn_skirt_far_left_master"(ComponentPose=@n27, Alpha=1, Translation="0,0,3", Rotation="0,0,0", Scale="1,1,1") {
    rotationMode=BMM_Additive
    translationMode=BMM_Additive
    translationSpace=BCS_BoneSpace
    alphaBoolBlend {blendOption=Linear}
    boneToModify {boneName=dyn_skirt_far_left_master}
    translation {z=3}
   }
   n29 = Control Rig(Source=Component To Local(ComponentPose=@n28)) {
    alphaBoolBlend {blendOption=Linear}
    controlRigAssetReference {blueprintRigClass=&/F_MED_Cube_Queen_CtrlRig.F_MED_Cube_Queen_CtrlRig_C}
   }
   n30 = Control Rig(Source=Blend Poses by bool(BlendPose_0=@n29, BlendPose_1=Component To Local(ComponentPose=@n8), BlendTime_0=0.2, BlendTime_1=0.2, bActiveValue=Get bIsCrouching()) {blendType=Linear}) {
    alphaBoolBlend {blendOption=Linear}
    controlRigAssetReference {blueprintRigClass=&/F_MED_Cube_Queen_Deform_CtrlRig.F_MED_Cube_Queen_Deform_CtrlRig_C}
   }
   n31 = "Transform (Modify) Bone - Bone: dyn_shoulder_piece_left"(ComponentPose=Local To Component(LocalPose=@n30), Alpha=Get ClavL() -> (ClavL:float), Translation="0,3,0", Rotation="0,5,0", Scale="1,1,1") {
    rotationMode=BMM_Additive
    rotationSpace=BCS_BoneSpace
    scaleMode=BMM_Additive
    scaleSpace=BCS_BoneSpace
    translationMode=BMM_Additive
    translationSpace=BCS_BoneSpace
    alphaBoolBlend {blendOption=Linear}
    boneToModify {boneName=dyn_shoulder_piece_left}
    rotation {yaw=5}
    translation {y=3}
   }
   n32 = "Transform (Modify) Bone - Bone: dyn_shoulder_piece_right"(ComponentPose=@n31, Alpha=Get ClavR() -> (ClavR:float), Translation="-2,-1,0", Rotation="10,-10,0", Scale="1,1,1") {
    rotationMode=BMM_Additive
    rotationSpace=BCS_BoneSpace
    scaleMode=BMM_Additive
    scaleSpace=BCS_BoneSpace
    translationMode=BMM_Additive
    translationSpace=BCS_BoneSpace
    alphaBoolBlend {blendOption=Linear}
    boneToModify {boneName=dyn_shoulder_piece_right}
    rotation {pitch=10, yaw=-10}
    translation=<-2 -1>
   }
   Output Pose(Result=Component To Local(ComponentPose=@n32)) {layerGroup=None}
   n33 = Get RigidBodyEnabledAlpha() -> (RigidBodyEnabledAlpha:float)
   n34 = "Constraint - dyn_skirt_far_left_master (TTTTT)"!(
    ComponentPose=Local To Component(LocalPose=Copy Pose From Mesh() {bUseAttachedParent=true})
    Alpha=@n33
    ConstraintWeights_0=0
    ConstraintWeights_1:float=0.4803
    ConstraintWeights_2:float=0.457
    ConstraintWeights_3:float=0.0001
    ConstraintWeights_4:float=0.0626
   ) -> (Pose) {
    alphaBoolBlend {blendOption=Linear}
    boneToModify {boneName=dyn_skirt_far_left_master}
    constraintSetup [
     {transformType=Translation, targetBone {boneName=thigh_r}}
     {transformType=Translation, targetBone {boneName=pelvis}}
     {transformType=Translation, targetBone {boneName=thigh_twist_01_l}}
     {transformType=Translation, targetBone {boneName=thigh_twist_01_r}}
     {transformType=Translation, targetBone {boneName=thigh_l}}
    ]
    constraintWeights [1,1,1,1,1]
   }
   n35 = "Constraint - dyn_skirt_far_right_master (TTTTT)"!(
    ComponentPose=@n34
    Alpha=@n33
    ConstraintWeights_0=0.0835
    ConstraintWeights_1:float=0.4244
    ConstraintWeights_2:float=0.0001
    ConstraintWeights_3:float=0.4921
    ConstraintWeights_4:float=0
   ) -> (Pose) {
    alphaBoolBlend {blendOption=Linear}
    boneToModify {boneName=dyn_skirt_far_right_master}
    constraintSetup [
     {transformType=Translation, targetBone {boneName=thigh_r}}
     {transformType=Translation, targetBone {boneName=pelvis}}
     {transformType=Translation, targetBone {boneName=thigh_twist_01_l}}
     {transformType=Translation, targetBone {boneName=thigh_twist_01_r}}
     {transformType=Translation, targetBone {boneName=thigh_l}}
    ]
    constraintWeights [1,1,1,1,1]
   }
   n36 = "Constraint - dyn_skirt_far_right_master (PPPPP)"!(
    ComponentPose=@n35
    Alpha=1
    ConstraintWeights_0=0.457
    ConstraintWeights_1:float=0.0626
    ConstraintWeights_2:float=0
    ConstraintWeights_3:float=0.0001
    ConstraintWeights_4:float=0.4803
   ) -> (Pose) {
    alphaBoolBlend {blendOption=Linear}
    boneToModify {boneName=dyn_skirt_far_right_master}
    constraintSetup [
     {targetBone {boneName=thigh_twist_01_r}}
     {targetBone {boneName=thigh_r}}
     {targetBone {boneName=thigh_l}}
     {targetBone {boneName=thigh_twist_01_l}}
     {targetBone {boneName=pelvis}}
    ]
    constraintWeights [1,1,1,1,1]
   }
   Save cached pose 'ChuggaCache'(Pose=Component To Local(ComponentPose=@n36)) {cacheName=ChuggaCache}
  ]}
 EventGraph {nodes [
   Event BlueprintUpdateAnimation()
   > n1 = Gravity Override State Function DEPRECATED(CharacterPartAnimInstance=Self-Reference(), GameTimeInSeconds=GetGameTimeInSeconds(), GravityOverrideParams=Get Gravity Override Params())
   > Set Gravity Override(Gravity Override=@n1)
   n2 = GetAnimInstance(self=GetOwnerSkeletalMeshComponent())
   > Set chuggaEmote(chuggaEmote:bool=OR Boolean(
     A=Montage_IsActive!(
      self:Anim Instance Object Reference=@n2
      Montage:Anim Montage Object Reference=/BRCosmetics/Animation/Game/MainPlayer/Emotes/Chugga/Emote_Chugga_Train_CMF_Frontend_M
     ) -> (ReturnValue:bool)
     B=Montage_IsActive(self=@n2, Montage=/BRCosmetics/Animation/Game/MainPlayer/Emotes/Chugga/Emote_Chugga_Train_CMF_M)
    )) -> (Output_Get:bool)
   Event OnBeginCrouching()
   > Delay(Duration=0.2)
   > ResetDynamics(InTeleportType=None)
   Event OnBeginPlayingEmote() {is=title K2Node_Event}
   > Delay(Duration=0.2)
   > ResetDynamics(InTeleportType=TeleportPhysics)
   > ResetDynamics(InTeleportType=ResetPhysics)
   Event OnEndPlayingEmote() {is=title K2Node_Event}
   > Delay(Duration=0.2)
   > ResetDynamics(InTeleportType=TeleportPhysics)
   > ResetDynamics(InTeleportType=ResetPhysics)
  ]}
]
variables [
 chuggaEmote {type=boolean}
 Gravity Override {type=Vector}
 Gravity Override Params {type=GravityOverrideParamsStruct, default {front_End {windAmplitude=<10 180 30>, windFrequency=<0.15 0.5 0.3>, windOffset=<0 -250 30>}}}
]
