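# Physics asset for the face accessory mesh of the F_MED_Cube_Queen player cosmetic, set up for dynamic hair.
# Capsules on a seven-bone main hair chain hang from the head with limited swing and twist, alongside two short split-top branches off the second bone.
# Capsules on head, neck, spine, pelvis and arms use free-motion constraints, likely as collision for the hair.
# A free constraint also links main bone three to the large split branch.
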PhysicsAsset /BRCosmetics/Characters/Player/Female/Medium/Bodies/F_MED_Cube_Queen/Meshes/Parts/F_MED_Cube_Queen_FaceAcc_Physics
bodies {
 clavicle_l [{shapeType=Capsule, length=12.6574, radius=7.5707, shapeName=clavicle_l_capsule, center=<10.521 0 0>, rotation=<rot 90 0 0>}]
 clavicle_r [{shapeType=Capsule, length=12.6574, radius=7.5707, shapeName=clavicle_l_capsule, center=<-12.821 0.0002 0>, rotation=<rot 90 0 0>}]
 dyn_hair_large_split_top_01 [{shapeType=Capsule, length=2.5102, radius=1.4898, shapeName=dyn_hair_large_split_top_01_capsule, center=<7.1157 0 0>, rotation=<rot 90 0 0>}]
 dyn_hair_large_split_top_02 [{shapeType=Capsule, length=6.1562, radius=1.3963, shapeName=dyn_hair_large_split_top_02_capsule, center=<4.7886 0 0>, rotation=<rot 90 0 0>}]
 dyn_hair_main_01 [{shapeType=Capsule, length=4.2416, radius=5.7584, center=<7.557 0 0>, rotation=<rot 90 0 0>}]
 dyn_hair_main_02 [{shapeType=Capsule, length=4.5559, radius=5.4441, center=<7.557 0 0>, rotation=<rot 90 0 0>}]
 dyn_hair_main_03 [{shapeType=Capsule, length=0.4425, radius=4.5285, center=<4.974 -0.0001 -0.509>, rotation=<rot 90 0 0>}]
 dyn_hair_main_04 [{shapeType=Capsule, length=0.5118, radius=4.5829, center=<4.8537 0 0>, rotation=<rot 90 0 0>}]
 dyn_hair_main_05 [{shapeType=Capsule, length=6.2361, radius=3.7639, center=<7.557 0 0>, rotation=<rot 90 0 0>}]
 dyn_hair_main_06 [{shapeType=Capsule, length=6.4387, radius=3.5612, center=<7.557 0 0>, rotation=<rot 90 0 0>}]
 dyn_hair_main_07 [{shapeType=Capsule, length=10.1819, radius=3.1335, shapeName=dyn_hair_main_07_capsule, center=<9.3289 0 0>, rotation=<rot 90 0 0>}]
 dyn_hair_main_split_top_01 [{shapeType=Capsule, length=3.0061, radius=1.3963, shapeName=dyn_hair_main_split_top_01_capsule, center=<5.246 0 0>, rotation=<rot 90 0 0>}]
 dyn_hair_main_split_top_02 [{
   shapeType=Capsule
   length=4.5332
   radius=1.3963
   shapeName=dyn_hair_main_split_top_02_capsule
   center=<3.9636 0.0001 0>
   rotation=<rot 90 0 0>
  }]
 head [{shapeType=Capsule, length=5, radius=5, shapeName=head_capsule, rotation=<rot 90 0 0>}]
 lowerarm_l [{shapeType=Capsule, length=18.1879, radius=6.0068, shapeName=lowerarm_r_capsule, center=<11.223 0.0009 0>, rotation=<rot 90 0 0>}]
 lowerarm_r [{shapeType=Capsule, length=18.1879, radius=6.0068, shapeName=lowerarm_r_capsule, center=<-13.1615 0.0001 0>, rotation=<rot 90 0 0>}]
 neck_01 [{shapeType=Capsule, length=5, radius=5, shapeName=neck_01_capsule, rotation=<rot 90 0 0>}]
 neck_02 [{shapeType=Capsule, length=5, radius=5, shapeName=neck_02_capsule, rotation=<rot 90 0 0>}]
 pelvis [{shapeType=Capsule, length=29.4925, radius=10.2824, shapeName=pelvis_capsule, center=<-5.1678 -6.5881 0>, rotation=<rot 0 -90 0>}]
 spine_01 [{shapeType=Capsule, length=18.0564, radius=8.3305, shapeName=spine_01_capsule, center=<3.9703 0 0.0002>, rotation=<rot 0 -90 0>}]
 spine_03 [{shapeType=Capsule, length=19.6847, radius=6.7022, shapeName=spine_01_capsule, center=<3.9703 0 0.0002>, rotation=<rot 0 -90 0>}]
 spine_05 [{shapeType=Capsule, length=19.255, radius=7.1319, shapeName=spine_01_capsule, center=<3.8898 0.0001 0.0002>, rotation=<rot 0 -90 0>}]
 upperarm_l [{shapeType=Capsule, length=17.6086, radius=6.5861, shapeName=upperarm_r_capsule, center=<11.223 0.0009 0>, rotation=<rot 90 0 0>}]
 upperarm_r [{shapeType=Capsule, length=17.6086, radius=6.5861, shapeName=upperarm_r_capsule, center=<-13.1615 0.0001 0>, rotation=<rot 90 0 0>}]
}
constraints [
 {bone1Name=dyn_hair_main_01, bone2Name=head, swing1LimitDegrees=20, swing2LimitDegrees=20, twistLimitDegrees=7}
 {bone1Name=dyn_hair_main_02, bone2Name=dyn_hair_main_01, swing1LimitDegrees=30, swing2LimitDegrees=30, twistLimitDegrees=7}
 {bone1Name=dyn_hair_main_03, bone2Name=dyn_hair_main_02, swing1LimitDegrees=45, swing2LimitDegrees=45, twistLimitDegrees=7}
 {bone1Name=dyn_hair_main_05, bone2Name=dyn_hair_main_04, swing1LimitDegrees=45, swing2LimitDegrees=45, twistLimitDegrees=7}
 {bone1Name=dyn_hair_main_06, bone2Name=dyn_hair_main_05, swing1LimitDegrees=45, swing2LimitDegrees=45, twistLimitDegrees=7}
 {bone1Name=dyn_hair_main_07, bone2Name=dyn_hair_main_06, swing1LimitDegrees=45, swing2LimitDegrees=45, twistLimitDegrees=7}
 {bone1Name=dyn_hair_main_04, bone2Name=dyn_hair_main_03, swing1LimitDegrees=45, swing2LimitDegrees=45, twistLimitDegrees=7}
 {
  bone1Name=head
  bone2Name=neck_02
  swing1LimitDegrees=45
  swing1Motion=Free
  swing2LimitDegrees=45
  swing2Motion=Free
  twistLimitDegrees=45
  twistMotion=Free
 }
 {
  bone1Name=spine_05
  bone2Name=spine_03
  swing1LimitDegrees=45
  swing1Motion=Free
  swing2LimitDegrees=45
  swing2Motion=Free
  twistLimitDegrees=45
  twistMotion=Free
 }
 {
  bone1Name=spine_03
  bone2Name=spine_01
  swing1LimitDegrees=45
  swing1Motion=Free
  swing2LimitDegrees=45
  swing2Motion=Free
  twistLimitDegrees=45
  twistMotion=Free
 }
 {
  bone1Name=spine_01
  bone2Name=pelvis
  swing1LimitDegrees=45
  swing1Motion=Free
  swing2LimitDegrees=45
  swing2Motion=Free
  twistLimitDegrees=45
  twistMotion=Free
 }
 {
  bone1Name=clavicle_l
  bone2Name=spine_05
  swing1LimitDegrees=45
  swing1Motion=Free
  swing2LimitDegrees=45
  swing2Motion=Free
  twistLimitDegrees=45
  twistMotion=Free
 }
 {
  bone1Name=clavicle_r
  bone2Name=spine_05
  swing1LimitDegrees=45
  swing1Motion=Free
  swing2LimitDegrees=45
  swing2Motion=Free
  twistLimitDegrees=45
  twistMotion=Free
 }
 {
  bone1Name=neck_01
  bone2Name=spine_05
  swing1LimitDegrees=45
  swing1Motion=Free
  swing2LimitDegrees=45
  swing2Motion=Free
  twistLimitDegrees=45
  twistMotion=Free
 }
 {
  bone1Name=neck_02
  bone2Name=neck_01
  swing1LimitDegrees=45
  swing1Motion=Free
  swing2LimitDegrees=45
  swing2Motion=Free
  twistLimitDegrees=45
  twistMotion=Free
 }
 {
  bone1Name=upperarm_l
  bone2Name=clavicle_l
  swing1LimitDegrees=45
  swing1Motion=Free
  swing2LimitDegrees=45
  swing2Motion=Free
  twistLimitDegrees=45
  twistMotion=Free
 }
 {
  bone1Name=lowerarm_l
  bone2Name=upperarm_l
  swing1LimitDegrees=45
  swing1Motion=Free
  swing2LimitDegrees=45
  swing2Motion=Free
  twistLimitDegrees=45
  twistMotion=Free
 }
 {
  bone1Name=upperarm_r
  bone2Name=clavicle_r
  swing1LimitDegrees=45
  swing1Motion=Free
  swing2LimitDegrees=45
  swing2Motion=Free
  twistLimitDegrees=45
  twistMotion=Free
 }
 {
  bone1Name=lowerarm_r
  bone2Name=upperarm_r
  swing1LimitDegrees=45
  swing1Motion=Free
  swing2LimitDegrees=45
  swing2Motion=Free
  twistLimitDegrees=45
  twistMotion=Free
 }
 {bone1Name=dyn_hair_main_split_top_01, bone2Name=dyn_hair_main_02, swing1LimitDegrees=7, swing2LimitDegrees=15, twistLimitDegrees=3}
 {
  bone1Name=dyn_hair_main_split_top_02
  bone2Name=dyn_hair_main_split_top_01
  swing1LimitDegrees=20
  swing2LimitDegrees=20
  twistLimitDegrees=5
 }
 {bone1Name=dyn_hair_large_split_top_01, bone2Name=dyn_hair_main_02, swing1LimitDegrees=40, swing2LimitDegrees=40, twistLimitDegrees=10}
 {
  bone1Name=dyn_hair_large_split_top_02
  bone2Name=dyn_hair_large_split_top_01
  swing1LimitDegrees=20
  swing2LimitDegrees=20
  twistLimitDegrees=5
 }
 {
  bone1Name=dyn_hair_main_03
  bone2Name=dyn_hair_large_split_top_01
  swing1LimitDegrees=45
  swing1Motion=Free
  swing2LimitDegrees=45
  swing2Motion=Free
  twistLimitDegrees=45
  twistMotion=Free
 }
]
