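# Post-process style AnimBP for the Cube Queen outfit's face accessory, driving its hair and floating debris from copied body and head poses.
# Weighted constraints attach split-hair master bones to head and main hair bones.
# Main hair switches between two RigidBody setups chosen by targeting, with parachute-dependent constraint, and a constrained, reshaped snap-back pose.
# Chained AnimDynamics springs then add small-strand and debris motion, most debris weighted by targeting.
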AnimBlueprint /BRCosmetics/Characters/Player/Female/Medium/Bodies/F_MED_Cube_Queen/Meshes/Parts/F_MED_Cube_Queen_FaceAcc_AnimBP
target_skeleton=&/F_MED_Cube_Queen_FaceAcc_Skeleton
graphs [
 AnimGraph {nodes [
   n1 = "Transform (Modify) Bone - Bone: dyn_root"(
    ComponentPose=Local To Component(LocalPose=Use cached pose 'Constrained_Pose'())
    Alpha=1
    Translation="0,0,0"
    Rotation=Get sim_space_rotation() -> (sim_space_rotation:Rotator)
    Scale="1,1,1"
   ) {
    rotationMode=BMM_Replace
    rotationSpace=BCS_WorldSpace
    translationSpace=BCS_BoneSpace
    alphaBoolBlend {blendOption=Linear}
    alphaScaleBiasClamp {bInterpResult=true, interpSpeedDecreasing=2, interpSpeedIncreasing=2}
    boneToModify {boneName=dyn_root}
   }
   n2 = RigidBody!(ComponentPose=@n1, Alpha, OverrideWorldGravity:Vector=Get gravity_override() -> (gravity_override:Vector), ExternalForce) -> (Pose) {
    bClampLinearTranslationLimitToRefPose=true
    bOverrideWorldGravity=true
    overridePhysicsAsset=&/F_MED_Cube_Queen_FaceAcc_Physics
    simulationSpace=BaseBoneSpace
    alphaBoolBlend {blendOption=Linear}
    baseBoneRef {boneName=dyn_root}
    componentAppliedLinearAccClamp=<2000 2000 2000>
    componentLinearAccScale=<0.35 0.35 0.45>
    componentLinearVelScale=<0.2 0.2 0.3>
    overrideWorldGravity {z=-980}
   }
   n3 = Get bIsParachuteOpened()
   n4 = "Constraint - dyn_hair_main_04 (P)"(ComponentPose=@n2, Alpha=float * float(A="To Float (Boolean)"(InBool=@n3), B=0.3), ConstraintWeights_0=1) {
    alphaBoolBlend {blendOption=Linear}
    boneToModify {boneName=dyn_hair_main_04}
    constraintSetup [{targetBone {boneName=spine_05}}]
    constraintWeights [1]
   }
   n5 = "Transform (Modify) Bone - Bone: dyn_root"(
    ComponentPose=Local To Component(LocalPose=Use cached pose 'Constrained_Pose'())
    Alpha=1
    Translation="0,0,0"
    Rotation=Get sim_space_rotation()
    Scale="1,1,1"
   ) {
    rotationMode=BMM_Replace
    rotationSpace=BCS_WorldSpace
    translationSpace=BCS_BoneSpace
    alphaBoolBlend {blendOption=Linear}
    alphaScaleBiasClamp {bInterpResult=true, interpSpeedDecreasing=2, interpSpeedIncreasing=2}
    boneToModify {boneName=dyn_root}
   }
   n6 = RigidBody(ComponentPose=@n5, OverrideWorldGravity=Get gravity_override()) {
    bClampLinearTranslationLimitToRefPose=true
    bOverrideWorldGravity=true
    overridePhysicsAsset=&/F_MED_Cube_Queen_FaceAcc_Alt_Physics
    simulationSpace=BaseBoneSpace
    alphaBoolBlend {blendOption=Linear}
    baseBoneRef {boneName=pelvis}
    componentAppliedLinearAccClamp=<2000 2000 2000>
    componentLinearAccScale=<0.35 0.35 0.45>
    componentLinearVelScale=<0.2 0.2 0.3>
    overrideWorldGravity {z=-980}
   }
   n7 = Blend Poses by bool(
    BlendPose_0=Component To Local(ComponentPose=@n6)
    BlendPose_1=Component To Local(ComponentPose=@n4)
    BlendTime_0=0.4
    BlendTime_1=0.4
    bActiveValue=Get bIsTargeting()
   ) {blendType=Linear}
   n8 = "Constraint - dyn_hair_main_01 (PP)"!(
    ComponentPose=Local To Component(LocalPose=Use cached pose 'Constrained_Pose'())
    Alpha=0.5
    ConstraintWeights_0=1
    ConstraintWeights_1:float=0
   ) -> (Pose) {
    alphaBoolBlend {blendOption=Linear}
    boneToModify {boneName=dyn_hair_main_01}
    constraintSetup [{targetBone {boneName=spine_05}}, {offsetOption=None, targetBone {boneName=spine_04}}]
    constraintWeights [1,1]
   }
   n9 = "Constraint - dyn_hair_main_02 (PP)"!(ComponentPose=@n8, Alpha=1, ConstraintWeights_0=0.5, ConstraintWeights_1:float=0.5) -> (Pose) {
    alphaBoolBlend {blendOption=Linear}
    boneToModify {boneName=dyn_hair_main_02}
    constraintSetup [{targetBone {boneName=neck_01}}, {targetBone {boneName=spine_05}}]
    constraintWeights [1,1]
   }
   n10 = "Constraint - dyn_hair_main_03 (PP)"!(ComponentPose=@n9, Alpha=1, ConstraintWeights_0=1, ConstraintWeights_1:float=0) -> (Pose) {
    alphaBoolBlend {blendOption=Linear}
    boneToModify {boneName=dyn_hair_main_03}
    constraintSetup [{targetBone {boneName=spine_05}}, {offsetOption=None, targetBone {boneName=spine_04}}]
    constraintWeights [1,1]
   }
   n11 = "Constraint - dyn_hair_main_04 (PP)"!(ComponentPose=@n10, Alpha=1, ConstraintWeights_0=1, ConstraintWeights_1:float=0) -> (Pose) {
    alphaBoolBlend {blendOption=Linear}
    boneToModify {boneName=dyn_hair_main_04}
    constraintSetup [{targetBone {boneName=spine_04}}, {offsetOption=None, targetBone {boneName=spine_05}}]
    constraintWeights [1,1]
   }
   n12 = "Constraint - dyn_hair_main_05 (PP)"!(ComponentPose=@n11, Alpha=1, ConstraintWeights_0=0.5, ConstraintWeights_1:float=0.5) -> (Pose) {
    alphaBoolBlend {blendOption=Linear}
    boneToModify {boneName=dyn_hair_main_05}
    constraintSetup [{targetBone {boneName=spine_03}}, {targetBone {boneName=spine_04}}]
    constraintWeights [1,1]
   }
   n13 = "Transform (Modify) Bone - Bone: dyn_hair_main_01"(ComponentPose=@n12, Alpha=1, Translation="0,0,0", Rotation="-8,0,0", Scale="1,1,1") {
    rotationMode=BMM_Additive
    rotationSpace=BCS_BoneSpace
    translationMode=BMM_Additive
    translationSpace=BCS_BoneSpace
    alphaBoolBlend {blendOption=Linear}
    boneToModify {boneName=dyn_hair_main_01}
    rotation {pitch=-8}
   }
   n14 = "Transform (Modify) Bone - Bone: dyn_hair_main_02"(ComponentPose=@n13, Alpha=1, Translation="0,0,0", Rotation="-10,0,0", Scale="1,1,1") {
    rotationMode=BMM_Additive
    rotationSpace=BCS_BoneSpace
    translationMode=BMM_Additive
    translationSpace=BCS_BoneSpace
    alphaBoolBlend {blendOption=Linear}
    boneToModify {boneName=dyn_hair_main_02}
    rotation {pitch=-10}
   }
   n15 = "Transform (Modify) Bone - Bone: dyn_hair_main_03"(ComponentPose=@n14, Alpha=1, Translation="0,0,0", Rotation="-15,0,0", Scale="1,1,1") {
    rotationMode=BMM_Additive
    rotationSpace=BCS_BoneSpace
    translationMode=BMM_Additive
    translationSpace=BCS_BoneSpace
    alphaBoolBlend {blendOption=Linear}
    boneToModify {boneName=dyn_hair_main_03}
    rotation {pitch=-15}
   }
   n16 = "Transform (Modify) Bone - Bone: dyn_hair_main_04"(ComponentPose=@n15, Alpha=1, Translation="0,0,0", Rotation="-10,0,0", Scale="1,1,1") {
    rotationMode=BMM_Additive
    rotationSpace=BCS_BoneSpace
    translationMode=BMM_Additive
    translationSpace=BCS_BoneSpace
    alphaBoolBlend {blendOption=Linear}
    boneToModify {boneName=dyn_hair_main_04}
    rotation {pitch=-10}
   }
   n17 = "Transform (Modify) Bone - Bone: dyn_hair_main_05"(ComponentPose=@n16, Alpha=1, Translation="0,0,0", Rotation="5,0,0", Scale="1,1,1") {
    rotationMode=BMM_Additive
    rotationSpace=BCS_BoneSpace
    translationMode=BMM_Additive
    translationSpace=BCS_BoneSpace
    alphaBoolBlend {blendOption=Linear}
    boneToModify {boneName=dyn_hair_main_05}
    rotation {pitch=5}
   }
   n18 = "Transform (Modify) Bone - Bone: dyn_hair_main_06"(ComponentPose=@n17, Alpha=1, Translation="0,0,0", Rotation="0,0,0", Scale="1,1,1") {
    rotationMode=BMM_Additive
    rotationSpace=BCS_BoneSpace
    translationMode=BMM_Additive
    translationSpace=BCS_BoneSpace
    alphaBoolBlend {blendOption=Linear}
    boneToModify {boneName=dyn_hair_main_06}
   }
   n19 = "Transform (Modify) Bone - Bone: dyn_hair_main_07"(ComponentPose=@n18, Alpha=1, Translation="0,0,0", Rotation="10,0,0", Scale="1,1,1") {
    rotationMode=BMM_Additive
    rotationSpace=BCS_BoneSpace
    translationMode=BMM_Additive
    translationSpace=BCS_BoneSpace
    alphaBoolBlend {blendOption=Linear}
    boneToModify {boneName=dyn_hair_main_07}
    rotation {pitch=10}
   }
   n20 = "Transform (Modify) Bone - Bone: dyn_hair_main_split_top_master"(ComponentPose=@n19, Alpha=1, Translation="0,0,0", Rotation="-10,0,0", Scale="1,1,1") {
    rotationMode=BMM_Additive
    rotationSpace=BCS_BoneSpace
    translationMode=BMM_Additive
    translationSpace=BCS_BoneSpace
    alphaBoolBlend {blendOption=Linear}
    boneToModify {boneName=dyn_hair_main_split_top_master}
    rotation {pitch=-10}
   }
   n21 = "Transform (Modify) Bone - Bone: dyn_hair_large_split_top_master"(ComponentPose=@n20, Alpha=1, Translation="0,0,0", Rotation="-10,0,0", Scale="1,1,1") {
    rotationMode=BMM_Additive
    rotationSpace=BCS_BoneSpace
    translationMode=BMM_Additive
    translationSpace=BCS_BoneSpace
    alphaBoolBlend {blendOption=Linear}
    boneToModify {boneName=dyn_hair_large_split_top_master}
    rotation {pitch=-10}
   }
   n22 = Anim Dynamics - Bone: dyn_hair_main_06(ComponentPose=@n21, Alpha=1, ExternalForce="0,0,0") {
    angularDampingOverride=200
    angularSpringConstant=500
    bAngularSpring=true
    bGravityOverrideInSimSpace=true
    bUseGravityOverride=true
    simulationSpace=BoneRelative
    alphaBoolBlend {blendOption=Linear}
    boundBone {boneName=dyn_hair_main_06}
    componentAppliedLinearAccClamp=<1500 1500 1500>
    componentLinearAccScale=<0.4 0.4 1>
    componentLinearVelScale=<0.2 0.2 0.4>
    gravityOverride {z=-980}
    physicsBodyDefinitions [{
      sphereCollisionRadius=0
      boundBone {boneName=dyn_hair_main_06}
      boxExtents=<16 16 16>
      constraintSetup {angularLimitsMax=<3 3 10>, angularLimitsMin=<-3 -15 -10>, angularTarget=<1 0 0>}
      localJointOffset=<2 0 0>
     }]
    relativeSpaceBone {boneName=root}
   }
   n23 = Anim Dynamics - Bone: dyn_hair_main_07(ComponentPose=@n22, Alpha=1, ExternalForce="0,0,0") {
    angularDampingOverride=200
    angularSpringConstant=500
    bAngularSpring=true
    bGravityOverrideInSimSpace=true
    bUseGravityOverride=true
    simulationSpace=BoneRelative
    alphaBoolBlend {blendOption=Linear}
    boundBone {boneName=dyn_hair_main_07}
    componentAppliedLinearAccClamp=<1500 1500 1500>
    componentLinearAccScale=<0.4 0.4 1>
    componentLinearVelScale=<0.2 0.2 0.4>
    gravityOverride {z=-980}
    physicsBodyDefinitions [{
      sphereCollisionRadius=0
      boundBone {boneName=dyn_hair_main_07}
      boxExtents=<16 16 16>
      constraintSetup {angularLimitsMax=<3 3 10>, angularLimitsMin=<-3 -15 -10>, angularTarget=<1 0 0>}
      localJointOffset=<2 0 0>
     }]
    relativeSpaceBone {boneName=root}
   }
   n24 = "Transform (Modify) Bone - Bone: dyn_hair_main_01"(ComponentPose=@n23, Alpha=1, Translation="0,0,0", Rotation="0,0,0", Scale="1,1,1") {
    rotationMode=BMM_Additive
    rotationSpace=BCS_BoneSpace
    scaleMode=BMM_Additive
    scaleSpace=BCS_BoneSpace
    translationMode=BMM_Additive
    translationSpace=BCS_BoneSpace
    alphaBoolBlend {blendOption=Linear}
    boneToModify {boneName=dyn_hair_main_01}
   }
   n25 = "Transform (Modify) Bone - Bone: dyn_hair_main_02"(ComponentPose=@n24, Alpha=1, Translation="0,0,-2", Rotation="0,0,0", Scale="1,1,1") {
    rotationMode=BMM_Additive
    rotationSpace=BCS_BoneSpace
    scaleMode=BMM_Additive
    scaleSpace=BCS_BoneSpace
    translationMode=BMM_Additive
    translationSpace=BCS_BoneSpace
    alphaBoolBlend {blendOption=Linear}
    boneToModify {boneName=dyn_hair_main_02}
    translation {z=-2}
   }
   n26 = "Transform (Modify) Bone - Bone: dyn_hair_main_03"(ComponentPose=@n25, Alpha=1, Translation="0,0,0", Rotation="0,0,0", Scale="1,1,1") {
    rotationMode=BMM_Additive
    rotationSpace=BCS_BoneSpace
    scaleMode=BMM_Additive
    scaleSpace=BCS_BoneSpace
    translationMode=BMM_Additive
    translationSpace=BCS_BoneSpace
    alphaBoolBlend {blendOption=Linear}
    boneToModify {boneName=dyn_hair_main_03}
   }
   n27 = "Transform (Modify) Bone - Bone: dyn_hair_main_04"(ComponentPose=@n26, Alpha=1, Translation="0,0,0", Rotation="0,0,0", Scale="0.9,1,1") {
    rotationMode=BMM_Additive
    rotationSpace=BCS_BoneSpace
    scaleMode=BMM_Additive
    scaleSpace=BCS_BoneSpace
    translationMode=BMM_Additive
    translationSpace=BCS_BoneSpace
    alphaBoolBlend {blendOption=Linear}
    boneToModify {boneName=dyn_hair_main_04}
    scale {x=0.9}
   }
   n28 = "Transform (Modify) Bone - Bone: dyn_hair_main_05"(ComponentPose=@n27, Alpha=1, Translation="0,0,0", Rotation="0,0,0", Scale="0.75,1,0.8") {
    rotationMode=BMM_Additive
    rotationSpace=BCS_BoneSpace
    scaleMode=BMM_Additive
    scaleSpace=BCS_BoneSpace
    translationMode=BMM_Additive
    translationSpace=BCS_BoneSpace
    alphaBoolBlend {blendOption=Linear}
    boneToModify {boneName=dyn_hair_main_05}
    scale {x=0.75, z=0.8}
   }
   n29 = "Transform (Modify) Bone - Bone: dyn_hair_main_06"(ComponentPose=@n28, Alpha=1, Translation="0,0,0", Rotation="0,0,0", Scale="0.75,1,0.8") {
    rotationMode=BMM_Additive
    rotationSpace=BCS_BoneSpace
    scaleMode=BMM_Additive
    scaleSpace=BCS_BoneSpace
    translationMode=BMM_Additive
    translationSpace=BCS_BoneSpace
    alphaBoolBlend {blendOption=Linear}
    boneToModify {boneName=dyn_hair_main_06}
    scale {x=0.75, z=0.8}
   }
   n30 = "Transform (Modify) Bone - Bone: dyn_hair_main_07"(ComponentPose=@n29, Alpha=1, Translation="0,0,0", Rotation="0,0,0", Scale="0.75,1,0.8") {
    rotationMode=BMM_Additive
    rotationSpace=BCS_BoneSpace
    scaleMode=BMM_Additive
    scaleSpace=BCS_BoneSpace
    translationMode=BMM_Additive
    translationSpace=BCS_BoneSpace
    alphaBoolBlend {blendOption=Linear}
    boneToModify {boneName=dyn_hair_main_07}
    scale {x=0.75, z=0.8}
   }
   n31 = "Transform (Modify) Bone - Bone: dyn_hair_main_split_top_master"(ComponentPose=@n30, Alpha=1, Translation="0,0,0", Rotation="-15,0,0", Scale="1,1,0.8") {
    rotationMode=BMM_Additive
    rotationSpace=BCS_BoneSpace
    scaleMode=BMM_Additive
    scaleSpace=BCS_BoneSpace
    translationMode=BMM_Additive
    translationSpace=BCS_BoneSpace
    alphaBoolBlend {blendOption=Linear}
    boneToModify {boneName=dyn_hair_main_split_top_master}
    rotation {pitch=-15}
    scale {z=0.8}
   }
   n32 = "Transform (Modify) Bone - Bone: dyn_hair_large_split_top_master"(ComponentPose=@n31, Alpha=1, Translation="0,0,0", Rotation="-10,0,0", Scale="1,1,0.8") {
    rotationMode=BMM_Additive
    rotationSpace=BCS_BoneSpace
    scaleMode=BMM_Additive
    scaleSpace=BCS_BoneSpace
    translationMode=BMM_Additive
    translationSpace=BCS_BoneSpace
    alphaBoolBlend {blendOption=Linear}
    boneToModify {boneName=dyn_hair_large_split_top_master}
    rotation {pitch=-10}
    scale {z=0.8}
   }
   n33 = Component To Local(ComponentPose=@n32)
   n34 = Get bSnapBack() -> (bSnapBack:bool)
   n35 = Anim Dynamics - Bone: dyn_hair_split_small_01(
    ComponentPose=Local To Component(LocalPose=Blend Poses by bool(BlendPose_0=@n33, BlendPose_1=@n7, bActiveValue=@n34) {blendType=Linear})
    Alpha=1
    ExternalForce="0,0,0"
   ) {
    angularDampingOverride=200
    angularSpringConstant=500
    bAngularSpring=true
    bGravityOverrideInSimSpace=true
    bUseGravityOverride=true
    simulationSpace=BoneRelative
    alphaBoolBlend {blendOption=Linear}
    boundBone {boneName=dyn_hair_split_small_01}
    componentAppliedLinearAccClamp=<1500 1500 1500>
    componentLinearAccScale=<0.4 0.4 1>
    componentLinearVelScale=<0.2 0.2 0.4>
    gravityOverride {z=-980}
    physicsBodyDefinitions [{
      sphereCollisionRadius=0
      boundBone {boneName=dyn_hair_split_small_01}
      boxExtents=<16 16 16>
      constraintSetup {angularLimitsMax=<3 10 10>, angularLimitsMin=<-3 -10 -10>, angularTarget=<1 0 0>}
      localJointOffset=<2 0 0>
     }]
    relativeSpaceBone {boneName=root}
   }
   n36 = Anim Dynamics - Bone: dyn_hair_split_small_02(ComponentPose=@n35, Alpha=1, ExternalForce="0,0,0") {
    angularDampingOverride=200
    angularSpringConstant=500
    bAngularSpring=true
    bGravityOverrideInSimSpace=true
    bUseGravityOverride=true
    simulationSpace=BoneRelative
    alphaBoolBlend {blendOption=Linear}
    boundBone {boneName=dyn_hair_split_small_02}
    componentAppliedLinearAccClamp=<1500 1500 1500>
    componentLinearAccScale=<0.4 0.4 1>
    componentLinearVelScale=<0.2 0.2 0.4>
    gravityOverride {z=-980}
    physicsBodyDefinitions [{
      sphereCollisionRadius=0
      boundBone {boneName=dyn_hair_split_small_02}
      boxExtents=<16 16 16>
      constraintSetup {angularLimitsMax=<3 10 10>, angularLimitsMin=<-3 -10 -10>, angularTarget=<1 0 0>}
      localJointOffset=<2 0 0>
     }]
    relativeSpaceBone {boneName=root}
   }
   n37 = Anim Dynamics - Bone: dyn_hair_split_small_03(ComponentPose=@n36, Alpha=1, ExternalForce="0,0,0") {
    angularDampingOverride=200
    angularSpringConstant=500
    bAngularSpring=true
    bGravityOverrideInSimSpace=true
    bUseGravityOverride=true
    simulationSpace=BoneRelative
    alphaBoolBlend {blendOption=Linear}
    boundBone {boneName=dyn_hair_split_small_03}
    componentAppliedLinearAccClamp=<1500 1500 1500>
    componentLinearAccScale=<0.4 0.4 1>
    componentLinearVelScale=<0.2 0.2 0.4>
    gravityOverride {z=-980}
    physicsBodyDefinitions [{
      sphereCollisionRadius=0
      boundBone {boneName=dyn_hair_split_small_03}
      boxExtents=<16 16 16>
      constraintSetup {angularLimitsMax=<3 10 10>, angularLimitsMin=<-3 -10 -10>, angularTarget=<1 0 0>}
      localJointOffset=<2 0 0>
     }]
    relativeSpaceBone {boneName=root}
   }
   n38 = Anim Dynamics - Bone: dyn_hair_split_small_04(ComponentPose=@n37, Alpha=1, ExternalForce="0,0,0") {
    angularDampingOverride=200
    angularSpringConstant=500
    bAngularSpring=true
    bGravityOverrideInSimSpace=true
    bUseGravityOverride=true
    simulationSpace=BoneRelative
    alphaBoolBlend {blendOption=Linear}
    boundBone {boneName=dyn_hair_split_small_04}
    componentAppliedLinearAccClamp=<1500 1500 1500>
    componentLinearAccScale=<0.4 0.4 1>
    componentLinearVelScale=<0.2 0.2 0.4>
    gravityOverride {z=-980}
    physicsBodyDefinitions [{
      sphereCollisionRadius=0
      boundBone {boneName=dyn_hair_split_small_04}
      boxExtents=<16 16 16>
      constraintSetup {angularLimitsMax=<3 10 10>, angularLimitsMin=<-3 -10 -10>, angularTarget=<1 0 0>}
      localJointOffset=<2 0 0>
     }]
    relativeSpaceBone {boneName=root}
   }
   n39 = Anim Dynamics - Bone: dyn_hair_split_small_05(ComponentPose=@n38, Alpha=1, ExternalForce="0,0,0") {
    angularDampingOverride=200
    angularSpringConstant=500
    bAngularSpring=true
    bGravityOverrideInSimSpace=true
    bUseGravityOverride=true
    simulationSpace=BoneRelative
    alphaBoolBlend {blendOption=Linear}
    boundBone {boneName=dyn_hair_split_small_05}
    componentAppliedLinearAccClamp=<1500 1500 1500>
    componentLinearAccScale=<0.4 0.4 1>
    componentLinearVelScale=<0.2 0.2 0.4>
    gravityOverride {z=-980}
    physicsBodyDefinitions [{
      sphereCollisionRadius=0
      boundBone {boneName=dyn_hair_split_small_05}
      boxExtents=<16 16 16>
      constraintSetup {angularLimitsMax=<3 10 10>, angularLimitsMin=<-3 -10 -10>, angularTarget=<1 0 0>}
      localJointOffset=<2 0 0>
     }]
    relativeSpaceBone {boneName=root}
   }
   n40 = Anim Dynamics - Bone: dyn_hair_split_small_06(ComponentPose=@n39, Alpha=1, ExternalForce="0,0,0") {
    angularDampingOverride=200
    angularSpringConstant=500
    bAngularSpring=true
    bGravityOverrideInSimSpace=true
    bUseGravityOverride=true
    simulationSpace=BoneRelative
    alphaBoolBlend {blendOption=Linear}
    boundBone {boneName=dyn_hair_split_small_06}
    componentAppliedLinearAccClamp=<1500 1500 1500>
    componentLinearAccScale=<0.4 0.4 1>
    componentLinearVelScale=<0.2 0.2 0.4>
    gravityOverride {z=-980}
    physicsBodyDefinitions [{
      sphereCollisionRadius=0
      boundBone {boneName=dyn_hair_split_small_06}
      boxExtents=<16 16 16>
      constraintSetup {angularLimitsMax=<3 10 10>, angularLimitsMin=<-3 -10 -10>, angularTarget=<1 0 0>}
      localJointOffset=<2 0 0>
     }]
    relativeSpaceBone {boneName=root}
   }
   n41 = Anim Dynamics - Bone: dyn_hair_split_small_07(ComponentPose=@n40, Alpha=1, ExternalForce="0,0,0") {
    angularDampingOverride=200
    angularSpringConstant=500
    bAngularSpring=true
    bGravityOverrideInSimSpace=true
    bUseGravityOverride=true
    simulationSpace=BoneRelative
    alphaBoolBlend {blendOption=Linear}
    boundBone {boneName=dyn_hair_split_small_07}
    componentAppliedLinearAccClamp=<1500 1500 1500>
    componentLinearAccScale=<0.4 0.4 1>
    componentLinearVelScale=<0.2 0.2 0.4>
    gravityOverride {z=-980}
    physicsBodyDefinitions [{
      sphereCollisionRadius=0
      boundBone {boneName=dyn_hair_split_small_07}
      boxExtents=<16 16 16>
      constraintSetup {angularLimitsMax=<3 10 10>, angularLimitsMin=<-3 -10 -10>, angularTarget=<1 0 0>}
      localJointOffset=<2 0 0>
     }]
    relativeSpaceBone {boneName=root}
   }
   n42 = Anim Dynamics - Bone: dyn_hair_split_small_08(ComponentPose=@n41, Alpha=1, ExternalForce="0,0,0") {
    angularDampingOverride=200
    angularSpringConstant=500
    bAngularSpring=true
    bGravityOverrideInSimSpace=true
    bUseGravityOverride=true
    simulationSpace=BoneRelative
    alphaBoolBlend {blendOption=Linear}
    boundBone {boneName=dyn_hair_split_small_08}
    componentAppliedLinearAccClamp=<1500 1500 1500>
    componentLinearAccScale=<0.4 0.4 1>
    componentLinearVelScale=<0.2 0.2 0.4>
    gravityOverride {z=-980}
    physicsBodyDefinitions [{
      sphereCollisionRadius=0
      boundBone {boneName=dyn_hair_split_small_08}
      boxExtents=<16 16 16>
      constraintSetup {angularLimitsMax=<3 10 10>, angularLimitsMin=<-3 -10 -10>, angularTarget=<1 0 0>}
      localJointOffset=<2 0 0>
     }]
    relativeSpaceBone {boneName=root}
   }
   n43 = Anim Dynamics - Bone: dyn_hair_split_small_09(ComponentPose=@n42, Alpha=1, ExternalForce="0,0,0") {
    angularDampingOverride=200
    angularSpringConstant=500
    bAngularSpring=true
    bGravityOverrideInSimSpace=true
    bUseGravityOverride=true
    simulationSpace=BoneRelative
    alphaBoolBlend {blendOption=Linear}
    boundBone {boneName=dyn_hair_split_small_09}
    componentAppliedLinearAccClamp=<1500 1500 1500>
    componentLinearAccScale=<0.4 0.4 1>
    componentLinearVelScale=<0.2 0.2 0.4>
    gravityOverride {z=-980}
    physicsBodyDefinitions [{
      sphereCollisionRadius=0
      boundBone {boneName=dyn_hair_split_small_09}
      boxExtents=<16 16 16>
      constraintSetup {angularLimitsMax=<3 10 10>, angularLimitsMin=<-3 -10 -10>, angularTarget=<1 0 0>}
      localJointOffset=<2 0 0>
     }]
    relativeSpaceBone {boneName=root}
   }
   n44 = Anim Dynamics - Bone: dyn_hair_split_small_10(ComponentPose=@n43, Alpha=1, ExternalForce="0,0,0") {
    angularDampingOverride=200
    angularSpringConstant=500
    bAngularSpring=true
    bGravityOverrideInSimSpace=true
    bUseGravityOverride=true
    simulationSpace=BoneRelative
    alphaBoolBlend {blendOption=Linear}
    boundBone {boneName=dyn_hair_split_small_10}
    componentAppliedLinearAccClamp=<1500 1500 1500>
    componentLinearAccScale=<0.4 0.4 1>
    componentLinearVelScale=<0.2 0.2 0.4>
    gravityOverride {z=-980}
    physicsBodyDefinitions [{
      sphereCollisionRadius=0
      boundBone {boneName=dyn_hair_split_small_10}
      boxExtents=<16 16 16>
      constraintSetup {angularLimitsMax=<3 10 10>, angularLimitsMin=<-3 -10 -10>, angularTarget=<1 0 0>}
      localJointOffset=<2 0 0>
     }]
    relativeSpaceBone {boneName=root}
   }
   n45 = Anim Dynamics - Bone: dyn_hair_split_small_11(ComponentPose=@n44, Alpha=1, ExternalForce="0,0,0") {
    angularDampingOverride=200
    angularSpringConstant=500
    bAngularSpring=true
    bGravityOverrideInSimSpace=true
    bUseGravityOverride=true
    simulationSpace=BoneRelative
    alphaBoolBlend {blendOption=Linear}
    boundBone {boneName=dyn_hair_split_small_11}
    componentAppliedLinearAccClamp=<1500 1500 1500>
    componentLinearAccScale=<0.4 0.4 1>
    componentLinearVelScale=<0.2 0.2 0.4>
    gravityOverride {z=-980}
    physicsBodyDefinitions [{
      sphereCollisionRadius=0
      boundBone {boneName=dyn_hair_split_small_11}
      boxExtents=<16 16 16>
      constraintSetup {angularLimitsMax=<3 10 10>, angularLimitsMin=<-3 -10 -10>, angularTarget=<1 0 0>}
      localJointOffset=<2 0 0>
     }]
    relativeSpaceBone {boneName=root}
   }
   n46 = Anim Dynamics - Bone: dyn_hair_split_small_12(ComponentPose=@n45, Alpha=1, ExternalForce="0,0,0") {
    angularDampingOverride=200
    angularSpringConstant=500
    bAngularSpring=true
    bGravityOverrideInSimSpace=true
    bUseGravityOverride=true
    simulationSpace=BoneRelative
    alphaBoolBlend {blendOption=Linear}
    boundBone {boneName=dyn_hair_split_small_12}
    componentAppliedLinearAccClamp=<1500 1500 1500>
    componentLinearAccScale=<0.4 0.4 1>
    componentLinearVelScale=<0.2 0.2 0.4>
    gravityOverride {z=-980}
    physicsBodyDefinitions [{
      sphereCollisionRadius=0
      boundBone {boneName=dyn_hair_split_small_12}
      boxExtents=<16 16 16>
      constraintSetup {angularLimitsMax=<3 10 10>, angularLimitsMin=<-3 -10 -10>, angularTarget=<1 0 0>}
      localJointOffset=<2 0 0>
     }]
    relativeSpaceBone {boneName=root}
   }
   n47 = Anim Dynamics - Bone: dyn_hair_split_small_13(ComponentPose=@n46, Alpha=1, ExternalForce="0,0,0") {
    angularDampingOverride=200
    angularSpringConstant=500
    bAngularSpring=true
    bGravityOverrideInSimSpace=true
    bUseGravityOverride=true
    simulationSpace=BoneRelative
    alphaBoolBlend {blendOption=Linear}
    boundBone {boneName=dyn_hair_split_small_13}
    componentAppliedLinearAccClamp=<1500 1500 1500>
    componentLinearAccScale=<0.4 0.4 1>
    componentLinearVelScale=<0.2 0.2 0.4>
    gravityOverride {z=-980}
    physicsBodyDefinitions [{
      sphereCollisionRadius=0
      boundBone {boneName=dyn_hair_split_small_13}
      boxExtents=<25 25 25>
      constraintSetup {angularLimitsMax=<3 10 10>, angularLimitsMin=<-3 -10 -10>, angularTarget=<1 0 0>}
      localJointOffset=<2 0 0>
     }]
    relativeSpaceBone {boneName=root}
   }
   n48 = Get IsTargeting() -> (IsTargeting:float)
   n49 = Anim Dynamics - Bone: dyn_face_debris_back_small_r(ComponentPose=@n47, Alpha=@n48, ExternalForce="0,0,0") {
    angularDampingOverride=1
    angularSpringConstant=20000
    bAngularSpring=true
    bGravityOverrideInSimSpace=true
    bLinearSpring=true
    bOverrideAngularDamping=true
    bOverrideLinearDamping=true
    bUseGravityOverride=true
    gravityScale=0.35
    linearDampingOverride=1
    linearSpringConstant=135
    simulationSpace=RootRelative
    alphaBoolBlend {blendOption=Linear}
    boundBone {boneName=dyn_face_debris_back_small_r}
    componentAppliedLinearAccClamp=<3000 3000 3000>
    componentLinearAccScale=<0.3 0.3 0.4>
    componentLinearVelScale=<0.3 0.3 0.4>
    gravityOverride {z=-980}
    physicsBodyDefinitions [{
      sphereCollisionRadius=0
      boundBone {boneName=dyn_face_debris_back_small_r}
      boxExtents=<20 30 30>
      constraintSetup {angularLimitsMax=<0 0 0>, angularLimitsMin=<0 0 0>, angularTarget=<1 0 0>, linearAxesMax=<0 50 50>, linearAxesMin=<-50 -50 -50>}
      localJointOffset=<0 0 0>
     }]
   }
   n50 = Anim Dynamics - Bone: dyn_face_debris_back_top_small(ComponentPose=@n49, Alpha=@n48, ExternalForce="0,0,0") {
    angularDampingOverride=1
    angularSpringConstant=20000
    bAngularSpring=true
    bGravityOverrideInSimSpace=true
    bLinearSpring=true
    bOverrideAngularDamping=true
    bOverrideLinearDamping=true
    bUseGravityOverride=true
    gravityScale=0.35
    linearDampingOverride=1
    linearSpringConstant=150
    simulationSpace=RootRelative
    alphaBoolBlend {blendOption=Linear}
    boundBone {boneName=dyn_face_debris_back_top_small}
    componentAppliedLinearAccClamp=<3000 3000 3000>
    componentLinearAccScale=<0.3 0.3 0.4>
    componentLinearVelScale=<0.3 0.3 0.4>
    gravityOverride {z=-980}
    physicsBodyDefinitions [{
      sphereCollisionRadius=0
      boundBone {boneName=dyn_face_debris_back_top_small}
      boxExtents=<20 30 30>
      constraintSetup {angularLimitsMax=<0 0 0>, angularLimitsMin=<0 0 0>, angularTarget=<1 0 0>, linearAxesMax=<0 50 50>, linearAxesMin=<-50 -50 -50>}
      localJointOffset=<0 0 0>
     }]
   }
   n51 = Anim Dynamics - Bone: dyn_face_debris_back_small_l(ComponentPose=@n50, Alpha=@n48, ExternalForce="0,0,0") {
    angularDampingOverride=1
    angularSpringConstant=20000
    bAngularSpring=true
    bGravityOverrideInSimSpace=true
    bLinearSpring=true
    bOverrideAngularDamping=true
    bOverrideLinearDamping=true
    bUseGravityOverride=true
    gravityScale=0.35
    linearDampingOverride=1
    linearSpringConstant=135
    simulationSpace=RootRelative
    alphaBoolBlend {blendOption=Linear}
    boundBone {boneName=dyn_face_debris_back_small_l}
    componentAppliedLinearAccClamp=<3000 3000 3000>
    componentLinearAccScale=<0.3 0.3 0.4>
    componentLinearVelScale=<0.3 0.3 0.4>
    gravityOverride {z=-980}
    physicsBodyDefinitions [{
      sphereCollisionRadius=0
      boundBone {boneName=dyn_face_debris_back_small_l}
      boxExtents=<20 30 30>
      constraintSetup {angularLimitsMax=<0 0 0>, angularLimitsMin=<0 0 0>, angularTarget=<1 0 0>, linearAxesMax=<0 50 50>, linearAxesMin=<-50 -50 -50>}
      localJointOffset=<0 0 0>
     }]
   }
   n52 = Anim Dynamics - Bone: dyn_face_debris_mid_l(ComponentPose=@n51, Alpha=@n48, ExternalForce="0,0,0") {
    angularDampingOverride=1
    angularSpringConstant=20000
    bAngularSpring=true
    bGravityOverrideInSimSpace=true
    bLinearSpring=true
    bOverrideAngularDamping=true
    bOverrideLinearDamping=true
    bUseGravityOverride=true
    gravityScale=0.35
    linearDampingOverride=1
    linearSpringConstant=115
    simulationSpace=RootRelative
    alphaBoolBlend {blendOption=Linear}
    boundBone {boneName=dyn_face_debris_mid_l}
    componentAppliedLinearAccClamp=<3000 3000 3000>
    componentLinearAccScale=<0.3 0.3 0.4>
    componentLinearVelScale=<0.3 0.3 0.4>
    gravityOverride {z=-980}
    physicsBodyDefinitions [{
      sphereCollisionRadius=0
      boundBone {boneName=dyn_face_debris_mid_l}
      boxExtents=<20 30 30>
      constraintSetup {angularLimitsMax=<0 0 0>, angularLimitsMin=<0 0 0>, angularTarget=<1 0 0>, linearAxesMax=<0 50 50>, linearAxesMin=<-50 -50 -50>}
      localJointOffset=<0 0 0>
     }]
   }
   n53 = Anim Dynamics - Bone: dyn_face_debris_mid_r(ComponentPose=@n52, Alpha=@n48, ExternalForce="0,0,0") {
    angularDampingOverride=1
    angularSpringConstant=20000
    bAngularSpring=true
    bGravityOverrideInSimSpace=true
    bLinearSpring=true
    bOverrideAngularDamping=true
    bOverrideLinearDamping=true
    bUseGravityOverride=true
    gravityScale=0.35
    linearDampingOverride=1
    linearSpringConstant=115
    simulationSpace=RootRelative
    alphaBoolBlend {blendOption=Linear}
    boundBone {boneName=dyn_face_debris_mid_r}
    componentAppliedLinearAccClamp=<3000 3000 3000>
    componentLinearAccScale=<0.3 0.3 0.4>
    componentLinearVelScale=<0.3 0.3 0.4>
    gravityOverride {z=-980}
    physicsBodyDefinitions [{
      sphereCollisionRadius=0
      boundBone {boneName=dyn_face_debris_mid_r}
      boxExtents=<20 30 30>
      constraintSetup {angularLimitsMax=<0 0 0>, angularLimitsMin=<0 0 0>, angularTarget=<1 0 0>, linearAxesMax=<0 50 50>, linearAxesMin=<-50 -50 -50>}
      localJointOffset=<0 0 0>
     }]
   }
   n54 = Anim Dynamics - Bone: dyn_face_debris_back_large_r(ComponentPose=@n53, Alpha=@n48, ExternalForce="0,0,0") {
    angularDampingOverride=1
    angularSpringConstant=20000
    bAngularSpring=true
    bGravityOverrideInSimSpace=true
    bLinearSpring=true
    bOverrideAngularDamping=true
    bOverrideLinearDamping=true
    bUseGravityOverride=true
    gravityScale=0.35
    linearDampingOverride=1
    linearSpringConstant=70
    simulationSpace=RootRelative
    alphaBoolBlend {blendOption=Linear}
    boundBone {boneName=dyn_face_debris_back_large_r}
    componentAppliedLinearAccClamp=<3000 3000 3000>
    componentLinearAccScale=<0.3 0.3 0.4>
    componentLinearVelScale=<0.3 0.3 0.4>
    gravityOverride {z=-980}
    physicsBodyDefinitions [{
      sphereCollisionRadius=0
      boundBone {boneName=dyn_face_debris_back_large_r}
      boxExtents=<20 30 30>
      constraintSetup {angularLimitsMax=<0 0 0>, angularLimitsMin=<0 0 0>, angularTarget=<1 0 0>, linearAxesMax=<0 50 50>, linearAxesMin=<-50 -50 -50>}
      localJointOffset=<0 0 0>
     }]
   }
   n55 = Anim Dynamics - Bone: dyn_face_debris_back_large_l(ComponentPose=@n54, Alpha=@n48, ExternalForce="0,0,0") {
    angularDampingOverride=1
    angularSpringConstant=20000
    bAngularSpring=true
    bGravityOverrideInSimSpace=true
    bLinearSpring=true
    bOverrideAngularDamping=true
    bOverrideLinearDamping=true
    bUseGravityOverride=true
    gravityScale=0.35
    linearDampingOverride=1
    linearSpringConstant=70
    simulationSpace=RootRelative
    alphaBoolBlend {blendOption=Linear}
    boundBone {boneName=dyn_face_debris_back_large_l}
    componentAppliedLinearAccClamp=<3000 3000 3000>
    componentLinearAccScale=<0.3 0.3 0.4>
    componentLinearVelScale=<0.3 0.3 0.4>
    gravityOverride {z=-980}
    physicsBodyDefinitions [{
      sphereCollisionRadius=0
      boundBone {boneName=dyn_face_debris_back_large_l}
      boxExtents=<20 30 30>
      constraintSetup {angularLimitsMax=<0 0 0>, angularLimitsMin=<0 0 0>, angularTarget=<1 0 0>, linearAxesMax=<0 50 50>, linearAxesMin=<-50 -50 -50>}
      localJointOffset=<0 0 0>
     }]
   }
   n56 = Anim Dynamics - Bone: dyn_face_debris_lower_r(ComponentPose=@n55, Alpha=@n48, ExternalForce="0,0,0") {
    angularDampingOverride=1
    angularSpringConstant=20000
    bAngularSpring=true
    bGravityOverrideInSimSpace=true
    bLinearSpring=true
    bOverrideAngularDamping=true
    bOverrideLinearDamping=true
    bUseGravityOverride=true
    gravityScale=0.35
    linearDampingOverride=1
    linearSpringConstant=150
    simulationSpace=RootRelative
    alphaBoolBlend {blendOption=Linear}
    boundBone {boneName=dyn_face_debris_lower_r}
    componentAppliedLinearAccClamp=<3000 3000 3000>
    componentLinearAccScale=<0.3 0.3 0.4>
    componentLinearVelScale=<0.3 0.3 0.4>
    gravityOverride {z=-980}
    physicsBodyDefinitions [{
      sphereCollisionRadius=0
      boundBone {boneName=dyn_face_debris_lower_r}
      boxExtents=<20 30 30>
      constraintSetup {angularLimitsMax=<0 0 0>, angularLimitsMin=<0 0 0>, angularTarget=<1 0 0>, linearAxesMax=<0 50 50>, linearAxesMin=<-50 -50 -50>}
      localJointOffset=<0 0 0>
     }]
   }
   n57 = Anim Dynamics - Bone: dyn_face_debris_lower_l(ComponentPose=@n56, Alpha=@n48, ExternalForce="0,0,0") {
    angularDampingOverride=1
    angularSpringConstant=20000
    bAngularSpring=true
    bGravityOverrideInSimSpace=true
    bLinearSpring=true
    bOverrideAngularDamping=true
    bOverrideLinearDamping=true
    bUseGravityOverride=true
    gravityScale=0.35
    linearDampingOverride=1
    linearSpringConstant=150
    simulationSpace=RootRelative
    alphaBoolBlend {blendOption=Linear}
    boundBone {boneName=dyn_face_debris_lower_l}
    componentAppliedLinearAccClamp=<3000 3000 3000>
    componentLinearAccScale=<0.3 0.3 0.4>
    componentLinearVelScale=<0.3 0.3 0.4>
    gravityOverride {z=-980}
    physicsBodyDefinitions [{
      sphereCollisionRadius=0
      boundBone {boneName=dyn_face_debris_lower_l}
      boxExtents=<20 30 30>
      constraintSetup {angularLimitsMax=<0 0 0>, angularLimitsMin=<0 0 0>, angularTarget=<1 0 0>, linearAxesMax=<0 50 50>, linearAxesMin=<-50 -50 -50>}
      localJointOffset=<0 0 0>
     }]
   }
   n58 = Anim Dynamics - Bone: dyn_face_debris_crown_small(ComponentPose=@n57, Alpha=0.1, ExternalForce="0,0,0") {
    angularDampingOverride=1
    angularSpringConstant=20000
    bAngularSpring=true
    bGravityOverrideInSimSpace=true
    bLinearSpring=true
    bOverrideAngularDamping=true
    bOverrideLinearDamping=true
    bUseGravityOverride=true
    gravityScale=0.35
    linearDampingOverride=1
    linearSpringConstant=150
    simulationSpace=RootRelative
    alphaBoolBlend {blendOption=Linear}
    boundBone {boneName=dyn_face_debris_crown_small}
    componentAppliedLinearAccClamp=<3000 3000 3000>
    componentLinearAccScale=<0.3 0.3 1>
    componentLinearVelScale=<0.3 0.3 1>
    gravityOverride {z=-980}
    physicsBodyDefinitions [{
      sphereCollisionRadius=0
      boundBone {boneName=dyn_face_debris_crown_small}
      boxExtents=<20 30 30>
      constraintSetup {angularLimitsMax=<0 0 0>, angularLimitsMin=<0 0 0>, angularTarget=<1 0 0>, linearAxesMax=<0 50 50>, linearAxesMin=<-50 -50 -50>}
      localJointOffset=<0 0 0>
     }]
   }
   n59 = "Transform (Modify) Bone - Bone: dyn_hair_main_01"!(ComponentPose=@n58, bAlphaBoolEnabled:bool=True, Translation="0,0,0", Rotation="0,0,0", Scale="1.000,1.000,0.700") -> (Pose) {
    alphaInputType=Bool
    rotationSpace=BCS_BoneSpace
    scaleMode=BMM_Additive
    scaleSpace=BCS_BoneSpace
    translationSpace=BCS_BoneSpace
    alphaBoolBlend {blendOption=Linear}
    boneToModify {boneName=dyn_hair_main_01}
    scale {z=0.7}
   }
   Output Pose(Result=Component To Local(ComponentPose=@n59)) {layerGroup=None}
   n60 = Layered blend per bone(
    BasePose=Copy Pose From Mesh(SourceMeshComponent=Get Body_Mesh()) {bCopyCurves=true, bUseAttachedParent=true}
    BlendPoses_0=Copy Pose From Mesh(SourceMeshComponent=Get Head_Mesh() -> (Head_Mesh:Skeletal Mesh Component Object Reference)) {bCopyCurves=true, bUseAttachedParent=true}
   ) {layerSetup [{branchFilters [{boneName=faceAttach}, {boneName=C_jaw}, {boneName=R_cheek_inner}, {boneName=L_cheek_inner}]}]}
   n61 = "Constraint - dyn_hair_split_small_01_master (PP)"!(ComponentPose=Local To Component(LocalPose=@n60), Alpha=1, ConstraintWeights_0=0.3819, ConstraintWeights_1:float=0.6181) -> (Pose) {
    alphaBoolBlend {blendOption=Linear}
    boneToModify {boneName=dyn_hair_split_small_01_master}
    constraintSetup [{targetBone {boneName=head}}, {targetBone {boneName=dyn_hair_main_01}}]
    constraintWeights [1,1]
   }
   n62 = "Constraint - dyn_hair_split_small_02_master (PP)"!(ComponentPose=@n61, Alpha=1, ConstraintWeights_0=0.1368, ConstraintWeights_1:float=0.8632) -> (Pose) {
    alphaBoolBlend {blendOption=Linear}
    boneToModify {boneName=dyn_hair_split_small_02_master}
    constraintSetup [{targetBone {boneName=head}}, {targetBone {boneName=dyn_hair_main_01}}]
    constraintWeights [1,1]
   }
   n63 = "Constraint - dyn_hair_split_small_03_master (PPP)"!(ComponentPose=@n62, Alpha=1, ConstraintWeights_0=0.0339, ConstraintWeights_1:float=0.2415, ConstraintWeights_2:float=0.7246) -> (Pose) {
    alphaBoolBlend {blendOption=Linear}
    boneToModify {boneName=dyn_hair_split_small_03_master}
    constraintSetup [{targetBone {boneName=head}}, {targetBone {boneName=dyn_hair_main_02}}, {targetBone {boneName=dyn_hair_main_01}}]
    constraintWeights [1,1,1]
   }
   n64 = "Constraint - dyn_hair_split_small_04_master (PPP)"!(ComponentPose=@n63, Alpha=1, ConstraintWeights_0=0.0297, ConstraintWeights_1:float=0.2246, ConstraintWeights_2:float=0.7457) -> (Pose) {
    alphaBoolBlend {blendOption=Linear}
    boneToModify {boneName=dyn_hair_split_small_04_master}
    constraintSetup [{targetBone {boneName=head}}, {targetBone {boneName=dyn_hair_main_02}}, {targetBone {boneName=dyn_hair_main_01}}]
    constraintWeights [1,1,1]
   }
   n65 = "Constraint - dyn_hair_split_small_05_master (PPP)"!(ComponentPose=@n64, Alpha=1, ConstraintWeights_0=0.1359, ConstraintWeights_1:float=0.1431, ConstraintWeights_2:float=0.721) -> (Pose) {
    alphaBoolBlend {blendOption=Linear}
    boneToModify {boneName=dyn_hair_split_small_05_master}
    constraintSetup [{targetBone {boneName=head}}, {targetBone {boneName=dyn_hair_main_02}}, {targetBone {boneName=dyn_hair_main_01}}]
    constraintWeights [1,1,1]
   }
   n66 = "Constraint - dyn_hair_split_small_06_master (PPPP)"!(
    ComponentPose=@n65
    Alpha=1
    ConstraintWeights_0=0.0155
    ConstraintWeights_1:float=0.1274
    ConstraintWeights_2:float=0.1322
    ConstraintWeights_3:float=0.7249
   ) -> (Pose) {
    alphaBoolBlend {blendOption=Linear}
    boneToModify {boneName=dyn_hair_split_small_06_master}
    constraintSetup [
     {targetBone {boneName=head}}
     {targetBone {boneName=dyn_hair_split_small_06}}
     {targetBone {boneName=dyn_hair_main_01}}
     {targetBone {boneName=dyn_hair_main_02}}
    ]
    constraintWeights [1,1,1,1]
   }
   n67 = "Constraint - dyn_hair_split_small_07_master (PPPP)"!(
    ComponentPose=@n66
    Alpha=1
    ConstraintWeights_0=0.0057
    ConstraintWeights_1:float=0.1667
    ConstraintWeights_2:float=0.8058
    ConstraintWeights_3:float=0.0217
   ) -> (Pose) {
    alphaBoolBlend {blendOption=Linear}
    boneToModify {boneName=dyn_hair_split_small_07_master}
    constraintSetup [
     {targetBone {boneName=head}}
     {targetBone {boneName=dyn_hair_main_03}}
     {targetBone {boneName=dyn_hair_main_02}}
     {targetBone {boneName=dyn_hair_main_01}}
    ]
    constraintWeights [1,1,1,1]
   }
   n68 = "Constraint - dyn_hair_split_small_08_master (PPPPPPP)"!(
    ComponentPose=@n67
    Alpha=1
    ConstraintWeights_0=0.0001
    ConstraintWeights_1:float=0.7702
    ConstraintWeights_2:float=0.0368
    ConstraintWeights_3:float=0.0053
    ConstraintWeights_4:float=0.0007
    ConstraintWeights_5:float=0.1195
    ConstraintWeights_6:float=0.0674
   ) -> (Pose) {
    alphaBoolBlend {blendOption=Linear}
    boneToModify {boneName=dyn_hair_split_small_08_master}
    constraintSetup [
     {targetBone {boneName=head}}
     {targetBone {boneName=dyn_hair_main_04}}
     {targetBone {boneName=dyn_hair_main_03}}
     {targetBone {boneName=dyn_hair_main_02}}
     {targetBone {boneName=dyn_hair_main_01}}
     {targetBone {boneName=dyn_hair_main_05}}
     {targetBone {boneName=dyn_hair_split_small_08}}
    ]
    constraintWeights [1,1,1,1,1,1,1]
   }
   n69 = "Constraint - dyn_hair_split_small_09_master (PP)"!(ComponentPose=@n68, Alpha=1, ConstraintWeights_0=0.2082, ConstraintWeights_1:float=0.7918) -> (Pose) {
    alphaBoolBlend {blendOption=Linear}
    boneToModify {boneName=dyn_hair_split_small_09_master}
    constraintSetup [{targetBone {boneName=dyn_hair_main_06}}, {targetBone {boneName=dyn_hair_main_05}}]
    constraintWeights [1,1]
   }
   n70 = "Constraint - dyn_hair_split_small_12_master (PPPP)"!(
    ComponentPose=@n69
    Alpha=1
    ConstraintWeights_0=0.3076
    ConstraintWeights_1:float=0.5596
    ConstraintWeights_2:float=0.0838
    ConstraintWeights_3:float=0.049
   ) -> (Pose) {
    alphaBoolBlend {blendOption=Linear}
    boneToModify {boneName=dyn_hair_split_small_12_master}
    constraintSetup [
     {targetBone {boneName=dyn_hair_main_07}}
     {targetBone {boneName=dyn_hair_main_06}}
     {targetBone {boneName=dyn_hair_split_small_12}}
     {targetBone {boneName=dyn_hair_main_05}}
    ]
    constraintWeights [1,1,1,1]
   }
   n71 = "Constraint - dyn_hair_main_split_top_master (PPPP)"!(
    ComponentPose=@n70
    Alpha=1
    ConstraintWeights_0=0.0023
    ConstraintWeights_1:float=0.1468
    ConstraintWeights_2:float=0.8313
    ConstraintWeights_3:float=0.0196
   ) -> (Pose) {
    alphaBoolBlend {blendOption=Linear}
    boneToModify {boneName=dyn_hair_main_split_top_master}
    constraintSetup [
     {targetBone {boneName=head}}
     {targetBone {boneName=dyn_hair_main_03}}
     {targetBone {boneName=dyn_hair_main_02}}
     {targetBone {boneName=dyn_hair_main_01}}
    ]
    constraintWeights [1,1,1,1]
   }
   n72 = "Constraint - dyn_hair_large_split_top_master (PPPP)"!(
    ComponentPose=@n71
    Alpha=1
    ConstraintWeights_0=0.0105
    ConstraintWeights_1:float=0.1132
    ConstraintWeights_2:float=0.7922
    ConstraintWeights_3:float=0.0842
   ) -> (Pose) {
    alphaBoolBlend {blendOption=Linear}
    boneToModify {boneName=dyn_hair_large_split_top_master}
    constraintSetup [
     {targetBone {boneName=head}}
     {targetBone {boneName=dyn_hair_main_03}}
     {targetBone {boneName=dyn_hair_main_02}}
     {targetBone {boneName=dyn_hair_main_01}}
    ]
    constraintWeights [1,1,1,1]
   }
   n73 = "Constraint - dyn_hair_large_split_bot_master (PPPPPP)"!(
    ComponentPose=@n72
    Alpha=1
    ConstraintWeights_0=0.0031
    ConstraintWeights_1:float=0.0482
    ConstraintWeights_2:float=0.738
    ConstraintWeights_3:float=0.0945
    ConstraintWeights_4:float=0.0163
    ConstraintWeights_5:float=0.1
   ) -> (Pose) {
    alphaBoolBlend {blendOption=Linear}
    boneToModify {boneName=dyn_hair_large_split_bot_master}
    constraintSetup [
     {targetBone {boneName=head}}
     {targetBone {boneName=dyn_hair_main_04}}
     {targetBone {boneName=dyn_hair_main_03}}
     {targetBone {boneName=dyn_hair_main_02}}
     {targetBone {boneName=dyn_hair_main_01}}
     {targetBone {boneName=dyn_hair_large_split_bot_01}}
    ]
    constraintWeights [1,1,1,1,1,1]
   }
   Save cached pose 'Constrained_Pose'(Pose=Component To Local(ComponentPose=@n73)) {cacheName=Constrained_Pose}
  ]}
 EventGraph {nodes [
   Event BlueprintUpdateAnimation()
   > n1 = Sequence!(execute) -> (then_0, then_1, then_2:Exec)
   n2 = Gravity Override State Function DEPRECATED(
    execute=@n1.then_1
    CharacterPartAnimInstance=Self-Reference()
    GameTimeInSeconds=GetGameTimeInSeconds()
    GravityOverrideParams=Get Gravity Override Params()
   )
   > Set gravity_override(gravity_override=@n2)
   > n3 = ComplexSimSpaceFunction DEPRECATED!(
    execute
    CharacterPartAnimInstance:Custom Character Part Anim Instance Object Reference=Self-Reference()
    InterpSpeedMax:double=9
    PawnSpeedMax:double=20
    RInterpErrorTolerance:double=1
    DeltaTime:double=0
    DebugDraw:bool=false
    sim_space_rotation_in:Rotator=Get sim_space_rotation()
    inComponentSpaceIn:bool=Get InComponentSpace() -> (InComponentSpace:bool)
    inComponentSpaceLastFrameIn:bool=Get InComponentSpaceLastFrame() -> (InComponentSpaceLastFrame:bool)
    lockedOnSimSpaceIn:bool=Get LockedOnSimSpace() -> (LockedOnSimSpace:bool)
    InterpSpeedIn:double=Get InterpSpeed() -> (InterpSpeed:double)
   ) -> (
    then
    sim_space_rotation:Rotator
    inComponentSpaceOut:bool
    inComponentSpaceLastFrameOut:bool
    lockedOnSimSpaceOut:bool
    InterpSpeedOut:double
   )
   > Set sim_space_rotation(sim_space_rotation:Rotator=@n3) -> (Output_Get:Rotator)
   > Set InComponentSpace(InComponentSpace:bool=@n3.inComponentSpaceOut) -> (Output_Get:bool)
   > Set InComponentSpaceLastFrame(InComponentSpaceLastFrame:bool=@n3.inComponentSpaceLastFrameOut) -> (Output_Get:bool)
   > Set LockedOnSimSpace(LockedOnSimSpace:bool=@n3.lockedOnSimSpaceOut) -> (Output_Get:bool)
   > Set InterpSpeed(InterpSpeed:double=@n3.InterpSpeedOut) -> (Output_Get:double)
   > Branch(Condition=Get bIsDBNOCarried())
   > Set gravity_override(gravity_override="0, 0,-980")
   Event OnCharacterCustomizationLoaded()
   n4 = CheckForBackbling()
   > Set isBackbling(isBackbling:double="To Float (Boolean)"(InBool=OR Boolean!(A=@n4, B=@n4.bIsBackblingACape, C:bool=@n4.bIsBackblingLong) -> (ReturnValue))) -> (Output_Get:double)
   > Set isCape(isCape:bool=@n4.bIsBackblingACape) -> (Output_Get:bool)
   Event BlueprintInitializeAnimation()
   > Set Body_Mesh(Body_Mesh=GetOwnerSkeletalMeshComponent())
   > n5 = Cast To PlayerPawn_Athena(Object=TryGetPawnOwner())
   > Set Head_Mesh(Head_Mesh:Skeletal Mesh Component Object Reference=GetSkeletalMeshForPartType(self=@n5, PartType=Head)) -> (Output_Get:Skeletal Mesh Component Object Reference)
  ]}
]
variables [
 Body_Mesh {type=SkeletalMeshComponent}
 Gravity Override Params {
  type=GravityOverrideParamsStruct
  default {
   front_End {windAmplitude=<10 180 30>, windFrequency=<0.15 1.5 0.55>, windOffset=<30 -300 -30>}
   skydive_FloatingDown {gravityOverride=<0 0 1000>}
   skydive_GoingUp {gravityOverride=<0 0 100>}
   swimming {gravityOverride=<0 0 0>}
  }
 }
 gravity_override {type=Vector}
 Head_Mesh {type=SkeletalMeshComponent}
 InComponentSpace {type=boolean}
 InComponentSpaceLastFrame {type=boolean}
 InterpSpeed {type=number}
 isBackbling {type=number}
 isCape {type=boolean}
 LockedOnSimSpace {type=boolean}
 parachute_alpha {type=number}
 sim_space_rotation {type=Rotator}
]
